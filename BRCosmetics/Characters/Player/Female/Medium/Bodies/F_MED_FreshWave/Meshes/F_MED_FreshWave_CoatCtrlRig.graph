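# Control Rig for the FreshWave female character's coat that corrects cape bones, and two fur bones, as the body moves.
# CoatFix parent-constrains cape base bones to thighs and pelvis, then layers ModifyTransforms weighted by a Delta Rotation function, which normalizes one swing or twist axis of thigh or neck rotation.
# Products of stored weights drive combination correctives; MoveForward, Zipline and a calf_r spherical pose reader weight others.
# An optional debug branch drives leg bones from controls.
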ControlRigBlueprint /BRCosmetics/Characters/Player/Female/Medium/Bodies/F_MED_FreshWave/Meshes/F_MED_FreshWave_CoatCtrlRig
graphs [
 ContainedGraph {nodes [
   n1 = RigVMFunctionEntryNode()
   SequenceAggregate_1 = &RigVMFunction_Sequence(ExecuteContext=@n1.ExecuteContext)
   SequenceAggregate_2 = &RigVMFunction_Sequence(ExecuteContext=@SequenceAggregate_1.B)
   SequenceAggregate_1_1 = &RigVMFunction_Sequence(ExecuteContext=@SequenceAggregate_2.B)
   SequenceAggregate_1_2 = &RigVMFunction_Sequence(ExecuteContext=@SequenceAggregate_1_1.B)
   RigVMFunctionReturnNode(B=@SequenceAggregate_1.A, E=@SequenceAggregate_1_1.A, G=@SequenceAggregate_1_2.A, H=@SequenceAggregate_1_2.B, C=@SequenceAggregate_2.A)
  ]}
 Multiply_3_1_2_1_ContainedGraph {nodes [
   n1 = RigVMFunctionEntryNode()
   RigVMFunctionReturnNode(Result=&RigVMFunction_MathDoubleMul(B=@n1.C, A=&RigVMFunction_MathDoubleMul(A=@n1.A, B=@n1.B).Result).Result)
  ]}
 New Function_ContainedGraph {nodes [
   RigVMFunctionEntryNode()
   > SequenceAggregate = RigVMAggregateNode()
   &RigUnit_ParentConstraint(
    AdvancedSettings=(InterpolationType=Average,RotationOrderForFilter=XZY)
    Child=(Type=Bone,Name="dyn_cape_p_l_base")
    Filter=(TranslationFilter=(bX=true,bY=true,bZ=true),RotationFilter=(bX=True,bY=True,bZ=True),ScaleFilter=(bX=True,bY=True,bZ=True))
    Parents=((Item=(Type=Bone,Name="thigh_l"),Weight=1))
    Weight=1
    ExecutePin=@SequenceAggregate.B
   )
   > &RigUnit_ParentConstraint(
    AdvancedSettings=(InterpolationType=Average,RotationOrderForFilter=XZY)
    Child=(Type=Bone,Name="dyn_cape_f_l_base")
    Parents=((Item=(Type=Bone,Name="thigh_l"),Weight=1))
    Weight=1
    bMaintainOffset=true
   )
   > &RigUnit_ParentConstraint(
    AdvancedSettings=(InterpolationType=Average,RotationOrderForFilter=XZY)
    Child=(Type=Bone,Name="dyn_cape_s_l_base")
    Filter=(TranslationFilter=(bX=True,bY=false,bZ=true),RotationFilter=(bX=True,bY=True,bZ=True),ScaleFilter=(bX=True,bY=True,bZ=True))
    Parents=((Item=(Type=Bone,Name="thigh_l"),Weight=1))
    Weight=1
    bMaintainOffset=true
   )
   > &RigUnit_ParentConstraint(
    AdvancedSettings=(InterpolationType=Average,RotationOrderForFilter=XZY)
    Child=(Type=Bone,Name="dyn_cape_b_l_base")
    Filter=(TranslationFilter=(bX=True,bY=false,bZ=true),RotationFilter=(bX=True,bY=True,bZ=True),ScaleFilter=(bX=True,bY=True,bZ=True))
    Parents=((Item=(Type=Bone,Name="dyn_cape_s_l_base"),Weight=0.8),(Item=(Type=Bone,Name="pelvis"),Weight=0.2))
    Weight=1
    bMaintainOffset=true
   )
   > &RigUnit_ParentConstraint(
    AdvancedSettings=(InterpolationType=Average,RotationOrderForFilter=XZY)
    Child=(Type=Bone,Name="dyn_cape_p_r_base")
    Filter=(TranslationFilter=(bX=True,bY=true,bZ=True),RotationFilter=(bX=True,bY=True,bZ=True),ScaleFilter=(bX=True,bY=True,bZ=True))
    Parents=((Item=(Type=Bone,Name="thigh_r"),Weight=1))
    Weight=1
   )
   > &RigUnit_ParentConstraint(
    AdvancedSettings=(InterpolationType=Average,RotationOrderForFilter=XZY)
    Child=(Type=Bone,Name="dyn_cape_f_r_base")
    Parents=((Item=(Type=Bone,Name="thigh_r"),Weight=1))
    Weight=1
   )
   > &RigUnit_ParentConstraint(
    AdvancedSettings=(InterpolationType=Average,RotationOrderForFilter=XZY)
    Child=(Type=Bone,Name="dyn_cape_s_r_base")
    Filter=(TranslationFilter=(bX=True,bY=true,bZ=True),RotationFilter=(bX=True,bY=True,bZ=True),ScaleFilter=(bX=True,bY=True,bZ=True))
    Parents=((Item=(Type=Bone,Name="thigh_r"),Weight=1))
    Weight=1
   )
   > &RigUnit_ParentConstraint(
    AdvancedSettings=(InterpolationType=Average,RotationOrderForFilter=XZY)
    Child=(Type=Bone,Name="dyn_cape_b_r_base")
    Filter=(TranslationFilter=(bX=True,bY=false,bZ=True),RotationFilter=(bX=True,bY=True,bZ=True),ScaleFilter=(bX=True,bY=True,bZ=True))
    Parents=((Item=(Type=Bone,Name="pelvis"),Weight=0.2),(Item=(Type=Bone,Name="dyn_cape_s_r_base"),Weight=0.8))
    Weight=1
   )
   Delta Rotation = RigVMFunctionReferenceNode(
    ParentType=Null
    child=thigh_l
    debugRotation=False
    max=-90
    onlyTwist=False
    parent=thigh_l_null
    primAxis=2
    twistAxis=(X=1,Y=0,Z=0)
    ExecuteContext=@SequenceAggregate.C
   )
   > &RigUnit_ModifyTransforms(
    ItemToModify="((Item=(Type=Bone,Name=\"dyn_cape_p_l_base\"),Transform=(Rotation=(X=-0.6226,Y=-0.1792,Z=0.3268,W=-0.6881),Translation=(X=11.6166,Y=-5.396,Z=0.2579),Scale3D=(X=0.75,Y=1,Z=1))),(Item=(Type=Bone,Name=\"dyn_cape_f_l_base\"),Transform=(Rotation=(X=0.3384,Y=0.0788, ...(truncated)"
    WeightMinimum=0
    Weight=@"Delta Rotation".ouputWeight
   )
   > RigVMVariableNode(Value=@"Delta Rotation".ouputWeight) {variable=legDeltaRot_SwingZup_l}
   > Delta Rotation_1 = RigVMFunctionReferenceNode(ParentType=Null, child=thigh_l, debugRotation=False, max=90, onlyTwist=False, parent=thigh_l_null, primAxis=2, twistAxis=(X=1,Y=0,Z=0))
   > &RigUnit_ModifyTransforms(
    ItemToModify="((Item=(Type=Bone,Name=\"dyn_cape_p_l_base\"),Transform=(Rotation=(X=0.0996,Y=-0.5649,Z=0.1422,W=0.8067),Translation=(X=-5,Y=-5,Z=-10),Scale3D=(X=1,Y=1,Z=1))),(Item=(Type=Bone,Name=\"dyn_cape_f_l_base\"),Transform=(Rotation=(X=0.0449,Y=-0.2549,Z=0.1677,W=0.951 ...(truncated)"
    WeightMinimum=0
    Weight=@"Delta Rotation_1".ouputWeight
   )
   > Delta Rotation_1_1 = RigVMFunctionReferenceNode(ParentType=Null, child=thigh_l, debugRotation=False, max=-90, onlyTwist=False, parent=thigh_l_null, primAxis=1, twistAxis=(X=1,Y=0,Z=0))
   > &RigUnit_ModifyTransforms(
    ItemToModify="((Item=(Type=Bone,Name=\"dyn_cape_p_l_base\"),Transform=(Rotation=(X=0,Y=0,Z=0,W=1),Translation=(X=10,Y=0,Z=0),Scale3D=(X=0.75,Y=1,Z=1))),(Item=(Type=Bone,Name=\"dyn_cape_f_l_base\"),Transform=(Rotation=(X=0.2404,Y=0.0524,Z=0.0121,W=0.9692),Translation=(X=15,Y ...(truncated)"
    WeightMinimum=0
    Weight=@"Delta Rotation_1_1".ouputWeight
   )
   > RigVMVariableNode(Value=@"Delta Rotation_1_1".ouputWeight) {variable=legDeltaRot_SwingYOut_l}
   > Delta Rotation_1_1_1 = RigVMFunctionReferenceNode(ParentType=Null, child=thigh_l, debugRotation=False, max=60, onlyTwist=False, parent=thigh_l_null, primAxis=1, twistAxis=(X=1,Y=0,Z=0))
   > &RigUnit_ModifyTransforms(
    ItemToModify="((Item=(Type=Bone,Name=\"dyn_cape_p_l_base\"),Transform=(Rotation=(X=-0.0552,Y=-0.1183,Z=-0.419,W=0.8986),Translation=(X=0,Y=0,Z=0),Scale3D=(X=1,Y=1,Z=1))),(Item=(Type=Bone,Name=\"dyn_cape_f_l_base\"),Transform=(Rotation=(X=-0.117,Y=-0.3214,Z=-0.3214,W=0.883), ...(truncated)"
    WeightMinimum=0
    Weight=@"Delta Rotation_1_1_1".ouputWeight
   )
   > RigVMVariableNode(Value=@"Delta Rotation_1_1_1".ouputWeight) {variable=legDeltaRot_SwingYin_l}
   > Delta Rotation_1_1_1_1 = RigVMFunctionReferenceNode(ParentType=Null, child=thigh_l, debugRotation=False, max=-45, onlyTwist=true, parent=thigh_l_null, primAxis=0, twistAxis=(X=1,Y=0,Z=0))
   > &RigUnit_ModifyTransforms(
    ItemToModify="((Item=(Type=Bone,Name=\"dyn_cape_p_l_base\"),Transform=(Rotation=(X=0,Y=0,Z=0,W=1),Translation=(X=0,Y=0,Z=0),Scale3D=(X=1,Y=1,Z=1))),(Item=(Type=Bone,Name=\"dyn_cape_f_l_base\"),Transform=(Rotation=(X=0,Y=0,Z=0,W=1),Translation=(X=0,Y=0,Z=0),Scale3D=(X=1,Y=1, ...(truncated)"
    WeightMinimum=0
    Weight=@"Delta Rotation_1_1_1_1".ouputWeight
   )
   > RigVMVariableNode() {variable=legDeltaRot_twistOut_l}
   Delta Rotation_2 = RigVMFunctionReferenceNode(
    ParentType=Null
    child=thigh_r
    debugRotation=False
    max=-90
    onlyTwist=False
    parent=thigh_r_null
    primAxis=2
    rotationOrder=YZX
    twistAxis=(X=-1,Y=0,Z=0)
    ExecuteContext=@SequenceAggregate.E
   )
   > &RigUnit_ModifyTransforms(
    ItemToModify="((Item=(Type=Bone,Name=\"dyn_cape_p_r_base\"),Transform=(Rotation=(X=0.0942,Y=0.2927,Z=-0.1884,W=0.9327),Translation=(X=-13.6286,Y=1.4214,Z=-4.1041),Scale3D=(X=0.75,Y=1,Z=1))),(Item=(Type=Bone,Name=\"dyn_cape_f_r_base\"),Transform=(Rotation=(X=-0.0698,Y=0.1223 ...(truncated)"
    WeightMinimum=0
    Weight=@"Delta Rotation_2".ouputWeight
   )
   > RigVMVariableNode(Value=@"Delta Rotation_2".ouputWeight) {variable=legDeltaRot_SwingZup_r}
   > Delta Rotation_1_2 = RigVMFunctionReferenceNode(
    ParentType=Null
    child=thigh_r
    debugRotation=False
    max=90
    onlyTwist=False
    parent=thigh_r_null
    primAxis=2
    rotationOrder=YZX
    twistAxis=(X=-1,Y=0,Z=0)
   )
   > &RigUnit_ModifyTransforms(
    ItemToModify="((Item=(Type=Bone,Name=\"dyn_cape_p_r_base\"),Transform=(Rotation=(X=0.0608,Y=0.5987,Z=-0.0473,W=-0.7972),Translation=(X=4.3554,Y=3.3035,Z=8.9644),Scale3D=(X=0.9,Y=1,Z=1))),(Item=(Type=Bone,Name=\"dyn_cape_f_r_base\"),Transform=(Rotation=(X=-0.0677,Y=-0.1209,Z ...(truncated)"
    WeightMinimum=0
    Weight=@"Delta Rotation_1_2".ouputWeight
   )
   > Delta Rotation_1_1_2 = RigVMFunctionReferenceNode(
    ParentType=Null
    child=thigh_r
    debugRotation=False
    max=-90
    onlyTwist=False
    parent=thigh_r_null
    primAxis=1
    rotationOrder=YZX
    twistAxis=(X=1,Y=0,Z=0)
   )
   > RigVMVariableNode(Value=@"Delta Rotation_1_1_2".ouputWeight) {variable=legDeltaRot_SwingYOut_r}
   n1 = RigVMVariableNode() {variable=legDeltaRot_SwingZup_r}
   > &RigUnit_ModifyTransforms(
    ItemToModify="((Item=(Type=Bone,Name=\"dyn_cape_p_r_base\"),Transform=(Rotation=(X=0,Y=0,Z=0,W=1),Translation=(X=-10,Y=0,Z=0),Scale3D=(X=1,Y=1,Z=1))),(Item=(Type=Bone,Name=\"dyn_cape_f_r_base\"),Transform=(Rotation=(X=0,Y=0,Z=0,W=1),Translation=(X=20,Y=5,Z=0),Scale3D=(X=1,Y ...(truncated)"
    WeightMinimum=0
    Weight=&RigVMFunction_MathDoubleMul(
     A=@"Delta Rotation_1_1_2".ouputWeight
     B=&RigVMFunction_MathDoubleRemap(SourceMaximum=1, SourceMinimum=0, TargetMaximum=0, TargetMinimum=1, bClamp=False, Value=@n1.Value).Result
    ).Result
   )
   > Delta Rotation_1_1_1_2 = RigVMFunctionReferenceNode(
    ParentType=Null
    child=thigh_r
    debugRotation=False
    max=60
    onlyTwist=False
    parent=thigh_r_null
    primAxis=1
    rotationOrder=YZX
    twistAxis=(X=1,Y=0,Z=0)
   )
   n2 = RigVMVariableNode() {variable=legDeltaRot_SwingZup_r}
   > &RigUnit_ModifyTransforms(
    ItemToModify="((Item=(Type=Bone,Name=\"dyn_cape_p_r_base\"),Transform=(Rotation=(X=0.0493,Y=0.1152,Z=0.432,W=-0.8931),Translation=(X=0,Y=0,Z=0),Scale3D=(X=1,Y=1,Z=1))),(Item=(Type=Bone,Name=\"dyn_cape_f_r_base\"),Transform=(Rotation=(X=-0.0093,Y=0.4257,Z=-0.161,W=-0.8904),T ...(truncated)"
    WeightMinimum=0
    Weight=&RigVMFunction_MathDoubleMul(
     A=@"Delta Rotation_1_1_1_2".ouputWeight
     B=&RigVMFunction_MathDoubleRemap(SourceMaximum=1, SourceMinimum=0, TargetMaximum=0.3, TargetMinimum=1, Value=@n2.Value).Result
    ).Result
   )
   > RigVMVariableNode() {variable=legDeltaRot_SwingYin_r}
   > Delta Rotation_1_1_1_1_1 = RigVMFunctionReferenceNode(
    ParentType=Null
    child=thigh_r
    debugRotation=False
    max=-45
    onlyTwist=true
    parent=thigh_r_null
    primAxis=0
    rotationOrder=YZX
    twistAxis=(X=1,Y=0,Z=0)
   )
   > &RigUnit_ModifyTransforms(
    ItemToModify="((Item=(Type=Bone,Name=\"dyn_cape_p_r_base\"),Transform=(Rotation=(X=0,Y=0,Z=0,W=1),Translation=(X=0,Y=0,Z=0),Scale3D=(X=1,Y=1,Z=1))),(Item=(Type=Bone,Name=\"dyn_cape_f_r_base\"),Transform=(Rotation=(X=0,Y=0,Z=0,W=1),Translation=(X=0,Y=0,Z=0),Scale3D=(X=1,Y=1, ...(truncated)"
    WeightMinimum=0
    Weight=@"Delta Rotation_1_1_1_1_1".ouputWeight
   )
   > RigVMVariableNode(Value=@"Delta Rotation_1_1_1_1_1".ouputWeight) {variable=legDeltaRot_twistOut_r}
   > &RigUnit_ModifyTransforms(
    ItemToModify="((Item=(Type=Bone,Name=\"dyn_cape_p_r_base\"),Transform=(Rotation=(X=0,Y=0,Z=0,W=1),Translation=(X=0,Y=0,Z=0),Scale3D=(X=1,Y=1,Z=1))),(Item=(Type=Bone,Name=\"dyn_cape_f_r_base\"),Transform=(Rotation=(X=0,Y=0,Z=0.0778,W=0.997),Translation=(X=0,Y=-6.775,Z=2.3141 ...(truncated)"
    WeightMinimum=0
    Weight=&RigVMFunction_MathDoubleMul(B=1, A=RigVMVariableNode() {variable=legDeltaRot_SwingZup_r}.Value).Result
   )
   Clamp = &RigVMFunction_MathDoubleClamp(Maximum=0.5, Value=RigVMVariableNode() {variable=legDeltaRot_twistOut_r}.Value)
   Multiply_3_1_2 = RigVMAggregateNode(
    B=@Clamp.Result
    A=RigVMVariableNode() {variable=legDeltaRot_SwingYin_r}.Value
    C=RigVMVariableNode() {variable=legDeltaRot_SwingZup_r}.Value
   )
   > &RigUnit_ModifyTransforms(
    ItemToModify="((Item=(Type=Bone,Name=\"dyn_cape_p_r_base\"),Transform=(Rotation=(X=0.0835,Y=0.0296,Z=-0.3324,W=0.939),Translation=(X=-1.6095,Y=0.0285,Z=1.6533),Scale3D=(X=1,Y=1,Z=1))),(Item=(Type=Bone,Name=\"dyn_cape_f_r_base\"),Transform=(Rotation=(X=0.0066,Y=-0.1211,Z=0.0 ...(truncated)"
    WeightMinimum=0
    Weight=@Multiply_3_1_2.Result
   )
   > &RigUnit_ModifyTransforms(
    ItemToModify="((Item=(Type=Bone,Name=\"dyn_cape_s_r_base\"),Transform=(Rotation=(X=0.0034,Y=-0.3183,Z=0.0011,W=0.948),Translation=(X=0,Y=0,Z=0),Scale3D=(X=1,Y=1,Z=1))),(Item=(Type=Bone,Name=\"dyn_cape_f_r_base\"),Transform=(Rotation=(X=0,Y=-0.2744,Z=0,W=0.9616),Translation= ...(truncated)"
    WeightMinimum=0
    Weight=&RigVMFunction_MathDoubleMul(B=RigVMVariableNode() {variable=legDeltaRot_twistOut_r}.Value, A=RigVMVariableNode() {variable=legDeltaRot_SwingZup_r}.Value).Result
   )
   > &RigUnit_ModifyTransforms(
    ItemToModify="((Item=(Type=Bone,Name=\"dyn_cape_p_l_base\"),Transform=(Rotation=(X=-0.1084,Y=0.066,Z=-0.0416,W=0.991),Translation=(X=-2.4503,Y=-1.9094,Z=0.0031),Scale3D=(X=1,Y=1,Z=1))),(Item=(Type=Bone,Name=\"dyn_cape_f_l_base\"),Transform=(Rotation=(X=0,Y=-0.0621,Z=0,W=0.9 ...(truncated)"
    WeightMinimum=0
    Weight=&RigVMFunction_MathDoubleMul(A=RigVMVariableNode() {variable=legDeltaRot_SwingZup_l}.Value, B=RigVMVariableNode() {variable=legDeltaRot_SwingYOut_l}.Value).Result
   )
   > &RigUnit_ModifyTransforms(
    ItemToModify="((Item=(Type=Bone,Name=\"base_cape_area_l_base\"),Transform=(Rotation=(X=-0.0872,Y=0,Z=0,W=0.9962),Translation=(X=0,Y=0,Z=0),Scale3D=(X=1,Y=1,Z=1))),(Item=(Type=Bone,Name=\"base_cape_area_r_base\"),Transform=(Rotation=(X=-0.0872,Y=0,Z=0,W=0.9962),Translation=( ...(truncated)"
    Weight=1
    WeightMinimum=0
   )
   > &RigUnit_ModifyTransforms(
    ItemToModify="((Item=(Type=Bone,Name=\"dyn_cape_b_l_base\"),Transform=(Rotation=(X=-0.0055,Y=-0.2119,Z=0.0394,W=0.9765),Translation=(X=0,Y=0,Z=0),Scale3D=(X=1,Y=1,Z=1))),(Item=(Type=Bone,Name=\"dyn_cape_b_r_base\"),Transform=(Rotation=(X=0.0089,Y=-0.2463,Z=-0.0634,W=0.9671) ...(truncated)"
    WeightMinimum=0
    Weight=RigVMVariableNode() {variable=MoveForward}.Value
   )
   > &RigUnit_ModifyTransforms(
    ItemToModify="((Item=(Type=Bone,Name=\"dyn_cape_b_l_base\"),Transform=(Rotation=(X=0,Y=-0.2602,Z=0,W=0.9656),Translation=(X=0,Y=0,Z=0),Scale3D=(X=1,Y=1,Z=1))),(Item=(Type=Bone,Name=\"dyn_cape_b_r_base\"),Transform=(Rotation=(X=0,Y=0.2606,Z=0,W=-0.9655),Translation=(X=0,Y=0, ...(truncated)"
    WeightMinimum=0
    Weight=RigVMVariableNode() {variable=Zipline}.Value
   )
   > n3 = &RigUnit_SphericalPoseReader(
    ActiveRegionScaleFactors=(PositiveWidth=1,NegativeWidth=1,PositiveHeight=1,NegativeHeight=1)
    ActiveRegionSize=0.1
    Debug=(bDrawDebug=true,bDraw2D=False,bDrawLocalAxes=False,DebugScale=25,DebugSegments=20,DebugThickness=0.25)
    DriverAxis=(X=1,Y=0,Z=0)
    DriverItem=(Type=Bone,Name="calf_r")
    FalloffRegionScaleFactors=(PositiveWidth=1,NegativeWidth=1,PositiveHeight=1,NegativeHeight=1)
    FalloffSize=0.2
    FlipHeightScaling=False
    FlipWidthScaling=False
    OptionalParentItem=(Type=Bone,Name="None")
    RotationOffset=(X=90,Y=0,Z=30)
   )
   > &RigUnit_ModifyTransforms(
    ItemToModify="((Item=(Type=Bone,Name=\"dyn_cape_b_l_base\"),Transform=(Rotation=(X=0,Y=0,Z=0.0872,W=0.9962),Translation=(X=0,Y=0,Z=0),Scale3D=(X=1,Y=1,Z=1))),(Item=(Type=Bone,Name=\"dyn_cape_b_r_base\"),Transform=(Rotation=(X=0.212,Y=-0.554,Z=0.1485,W=0.7912),Translation=(X ...(truncated)"
    WeightMinimum=0
    Weight=@n3.OutputParam
   )
   Delta Rotation_2_1 = RigVMFunctionReferenceNode(
    ParentType=Bone
    child=neck_02
    debugRotation=False
    max=40
    min=15
    onlyTwist=False
    parent=head
    primAxis=2
    rotationOrder=ZYX
    twistAxis=(X=1,Y=0,Z=0)
    ExecuteContext=@SequenceAggregate.G
   )
   > &RigUnit_ModifyTransforms(
    ItemToModify=((Item=(Type=Bone,Name="base_fur_01_l"),Transform=(Rotation=(X=0,Y=0,Z=0,W=1),Translation=(X=8,Y=0,Z=0),Scale3D=(X=1,Y=1,Z=1))),(Item=(Type=Bone,Name="dyn_fur_01_r"),Transform=(Rotation=(X=0,Y=0,Z=0,W=1),Translation=(X=-8,Y=0,Z=0),Scale3D=(X=1,Y=1,Z=1))))
    WeightMinimum=0
    Weight=@"Delta Rotation_2_1".ouputWeight
   )
   RigVMFunctionReturnNode(ExecuteContext=@SequenceAggregate.H)
  ]}
 New Function_ContainedGraph {nodes [
   n1 = RigVMFunctionEntryNode()
   Clamp = &RigVMFunction_MathIntClamp(Value=@n1.primAxis)
   RigVMFunction_MathBoolToInteger = &RigVMFunction_MathBoolToInteger(Value=@n1.onlyTwist)
   RigVMFunction_MathQuaternionSwingTwist = &RigVMFunction_MathQuaternionSwingTwist(
    TwistAxis=@n1.twistAxis
    Input=&RigUnit_GetRelativeTransformForItem(
     Child=(Type=Bone,Name="None")
     Parent=(Type=Bone,Name="None")
     bChildInitial=False
     bParentInitial=False
     Parent.Type=@n1.ParentType
     Child.Name=@n1.child
     Parent.Name=@n1.parent
    ).RelativeTransform.Rotation
   )
   RigVMFunction_MathQuaternionToEuler = &RigVMFunction_MathQuaternionToEuler(
    Value=RigVMDispatchNode(
     Values=((),())
     Index=@RigVMFunction_MathBoolToInteger.Result
     Values.0=@RigVMFunction_MathQuaternionSwingTwist.Swing
     Values.1=@RigVMFunction_MathQuaternionSwingTwist.Twist
    ) {template="DISPATCH_RigVMDispatch_SelectInt32(in Index,in Values,out Result)"}.Result
    RotationOrder=@n1.rotationOrder
   )
   DISPATCH_RigVMDispatch_SelectInt32 = RigVMDispatchNode(
    Values=(0,0,0)
    Index=@Clamp.Result
    Values.0=@RigVMFunction_MathQuaternionToEuler.Result.X
    Values.1=@RigVMFunction_MathQuaternionToEuler.Result.Y
    Values.2=@RigVMFunction_MathQuaternionToEuler.Result.Z
   ) {template="DISPATCH_RigVMDispatch_SelectInt32(in Index,in Values,out Result)"}
   Remap = &RigVMFunction_MathDoubleRemap(
    TargetMaximum=1
    TargetMinimum=0
    bClamp=False
    Value=@DISPATCH_RigVMDispatch_SelectInt32.Result
    SourceMaximum=@n1.max
    SourceMinimum=@n1.min
   )
   > RigVMFunctionReturnNode(ouputWeight=&RigVMFunction_MathDoubleClamp(Value=@Remap.Result).Result)
  ]}
 New Function_ContainedGraph {nodes [
   n1 = RigVMFunctionEntryNode()
   RigVMFunction_ControlFlowBranch = &RigVMFunction_ControlFlowBranch(Condition=@n1.Debug Controls, ExecuteContext=@n1.ExecuteContext)
   RigVMFunctionReturnNode(ExecuteContext=@RigVMFunction_ControlFlowBranch.Completed)
   RigUnit_GetTransform = &RigUnit_GetTransform(Item=(Type=Control,Name="thigh_l_ctrl"), bInitial=False)
   RigUnit_SetTransform = &RigUnit_SetTransform(
    Item=(Type=Bone,Name="thigh_l")
    Weight=1
    bInitial=False
    Value=@RigUnit_GetTransform.Transform
    ExecutePin=@RigVMFunction_ControlFlowBranch.True
   )
   > &RigUnit_SetTransform(
    Item=(Type=Bone,Name="thigh_r")
    Weight=1
    bInitial=False
    Value=&RigUnit_GetTransform(Item=(Type=Control,Name="thigh_r_ctrl"), bInitial=False).Transform
   )
   > &RigUnit_SetTransform(
    Item=(Type=Bone,Name="calf_l")
    Weight=1
    bInitial=False
    Value=&RigUnit_GetTransform(Item=(Type=Control,Name="calf_l_ctrl"), bInitial=False).Transform
   )
   > &RigUnit_SetTransform(
    Item=(Type=Bone,Name="calf_r")
    Weight=1
    bInitial=False
    Value=&RigUnit_GetTransform(Item=(Type=Control,Name="calf_r_ctrl"), bInitial=False).Transform
   )
  ]}
 RigVMFunctionLibrary {nodes [
   CoatFix = RigVMCollapseNode(ExecuteContext=())
   Controls = RigVMCollapseNode(Debug Controls=False, ExecuteContext=())
   Delta Rotation = RigVMCollapseNode(ExecuteContext=(), debugRotation=False, onlyTwist=False, twistAxis=(X=0,Y=0,Z=0))
  ]}
 RigVMModel {nodes [
   RigUnit_BeginExecution = &RigUnit_BeginExecution()
   Controls = RigVMFunctionReferenceNode(Debug Controls=false, ExecuteContext=@RigUnit_BeginExecution.ExecutePin)
   > CoatFix = RigVMFunctionReferenceNode()
  ]}
]
hierarchy [
 root {loc=<0 0 0>, rot=<0 0 0 1>}
 pelvis {parent=root, loc=<0 -0.5301 98.0571>, rot=<0.0127 0.707 -0.0127 -0.707>}
 spine_01 {parent=pelvis, loc=<6.1208 -0.9328 0>, rot=<0 0 0.034 -0.9994>}
 spine_02 {parent=spine_01, loc=<6.5996 -0.2095 0>, rot=<0 0 0.0238 -0.9997>}
 spine_03 {parent=spine_02, loc=<6.4914 -0.1237 0>, rot=<0 0 0.0149 -0.9999>}
 spine_04 {parent=spine_03, loc=<6.9598 0.0794 0>, rot=<0 0 -0.0132 -0.9999>}
 spine_05 {parent=spine_04, loc=<6.9987 0 0>, rot=<0 0 -0.0015 -1>}
 attach_cape {parent=spine_05, loc=<9.0082 -11.6912 -0.0211>, rot=<0.0616 -0.7044 -0.0616 -0.7044>}
 clavicle_l {parent=spine_05, loc=<5.4831 -0.1053 -3.2946>, rot=<0.0465 -0.7289 -0.009 -0.6829>}
 upperarm_l {parent=clavicle_l, loc=<9.8658 -1.2136 -0.4105>, rot=<0.0164 -0.3245 -0.0269 -0.9454>}
 lowerarm_l {parent=upperarm_l, loc=<26.465 -0.8345 0.259>, rot=<0.0036 0.0161 -0.2298 -0.9731>}
 hand_l {parent=lowerarm_l, loc=<22.7355 -0.3203 -0.6105>, rot=<-0.6784 -0.0223 0.0735 -0.7307>}
 thumb_01_l {parent=hand_l, loc=<1.3118 -0.3191 -2.459>, rot=<0.6546 -0.4042 0.0664 -0.6354>}
 thumb_02_l {parent=thumb_01_l, loc=<5.9336 -0.0448 0.0401>, rot=<-0.0037 0.0213 0.298 -0.9543>}
 thumb_03_l {parent=thumb_02_l, loc=<2.2223 0.3682 -0.0907>, rot=<0.0118 -0.013 0.0066 -0.9998>}
 index_metacarpal_l {parent=hand_l, loc=<1.6535 0.0183 -2.96>, rot=<0.0158 -0.0124 -0.0173 -0.9996>}
 index_01_l {parent=index_metacarpal_l, loc=<8.3008 0.0357 -0.0524>, rot=<-0.0044 0.0485 0.0924 -0.9945>}
 index_02_l {parent=index_01_l, loc=<3.3587 -0.3004 0.0452>, rot=<-0.002 0.0027 0.3225 -0.9466>}
 index_03_l {parent=index_02_l, loc=<2.5578 -0.123 -0.0252>, rot=<-0.002 0.0067 0.2558 -0.9667>}
 middle_metacarpal_l {parent=hand_l, loc=<2.1856 -0.1725 -0.5984>, rot=<0.006 0.0168 -0.0329 -0.9993>}
 middle_01_l {parent=middle_metacarpal_l, loc=<7.3275 -0.2447 0.0129>, rot=<-0.0208 0.0042 0.0858 -0.9961>}
 middle_02_l {parent=middle_01_l, loc=<4.0123 0.1401 0.1175>, rot=<-0.0218 0.0258 0.3512 -0.9357>}
 middle_03_l {parent=middle_02_l, loc=<2.6077 0.0534 0.1373>, rot=<0.0082 0.0188 0.1708 -0.9851>}
 ring_metacarpal_l {parent=hand_l, loc=<2.2074 -0.3625 1.3638>, rot=<-0.0284 0.0722 -0.0458 -0.9959>}
 ring_01_l {parent=ring_metacarpal_l, loc=<7.0295 -0.6197 -0.1446>, rot=<-0.0489 -0.0252 0.1892 -0.9804>}
 ring_02_l {parent=ring_01_l, loc=<3.6174 0.4647 -0.1027>, rot=<0 0 0.2806 -0.9598>}
 ring_03_l {parent=ring_02_l, loc=<2.4152 0.114 0.0199>, rot=<0 0 0.1189 -0.9929>}
 pinky_metacarpal_l {parent=hand_l, loc=<1.7616 -0.9753 3.3814>, rot=<-0.0397 0.0814 -0.069 -0.9935>}
 pinky_01_l {parent=pinky_metacarpal_l, loc=<7.1709 -0.4831 -0.1094>, rot=<-0.0654 -0.0035 0.2242 -0.9723>}
 pinky_02_l {parent=pinky_01_l, loc=<2.8274 0.0893 -0.1043>, rot=<0.0037 0.0011 0.2854 -0.9584>}
 pinky_03_l {parent=pinky_02_l, loc=<2.2046 0.0307 -0.0078>, rot=<-0.0002 0.0098 0.1445 -0.9895>}
 deform_wrist_dn_base_l {parent=hand_l, loc=<0 -1.5326 0>, rot=<0 0 0 -1>}
 deform_wrist_dn_l {parent=deform_wrist_dn_base_l, loc=<0 0 0>, rot=<0 0 0 1>}
 deform_wrist_up_base_l {parent=hand_l, loc=<0 2.3009 0>, rot=<0 0 0 -1>}
 deform_wrist_up_l {parent=deform_wrist_up_base_l, loc=<0 0 0>, rot=<0 0 0 1>}
 lowerarm_twist_02_l {parent=lowerarm_l, loc=<10.6707 0 0>, rot=<0 0 0 1>}
 lowerarm_twist_01_l {parent=lowerarm_l, loc=<18.316 0 0>, rot=<0 0 0 1>}
 sleeve_l {parent=lowerarm_twist_01_l, loc=<0 0 0>, rot=<0 0 0 1>}
 sleeves_l_01 {parent=lowerarm_l, loc=<4.8179 3.3608 -5.4591>, rot=<-0.0768 -0.7193 0.0681 -0.6871>}
 sleeves_l_02 {parent=lowerarm_l, loc=<13.7878 1.5108 -8.0713>, rot=<-0.0768 -0.7193 0.0681 -0.6871>}
 deform_elbow_in_base_l {parent=lowerarm_l, loc=<0.1834 3.6062 -0.2572>, rot=<0 0 0 -1>}
 deform_elbow_in_l {parent=deform_elbow_in_base_l, loc=<0 0 0>, rot=<0 0 0 1>}
 deform_elbow_out_base_l {parent=lowerarm_l, loc=<0 -4 0>, rot=<0 0 0 1>}
 deform_elbow_out_l {parent=deform_elbow_out_base_l, loc=<0 0 0>, rot=<0 0 0 1>}
 upperarm_twist_01_l {parent=upperarm_l, loc=<7.2135 -0.1134 0>, rot=<0 0 0 1>}
 deform_upperarm_bk_base_l {parent=upperarm_twist_01_l, loc=<-6.4053 -4.3746 0.314>, rot=<0 0 0 -1>}
 deform_upperarm_bk_l {parent=deform_upperarm_bk_base_l, loc=<0 0 0>, rot=<0 0 0 1>}
 deform_upperarm_fr_base_l {parent=upperarm_twist_01_l, loc=<-5.6312 5.092 0.3354>, rot=<0 0 0 -1>}
 deform_upperarm_fr_l {parent=deform_upperarm_fr_base_l, loc=<0 0 0>, rot=<0 0 0 1>}
 deform_upperarm_up_base_l {parent=upperarm_twist_01_l, loc=<-8.891 0.4825 3.9517>, rot=<0 0 0 -1>}
 deform_upperarm_up_l {parent=deform_upperarm_up_base_l, loc=<0 0 0>, rot=<0 0 0 1>}
 upperarm_twist_02_l {parent=upperarm_l, loc=<19.7998 -0.1134 0>, rot=<0 0 0 1>}
 deform_upperarm_lw_l {parent=upperarm_l, loc=<10.8352 -0.0235 -6.6242>, rot=<0.5271 -0.007 0.8498 -0.0028>}
 attach_backpack {parent=spine_05, loc=<-0.1136 -13.4387 -0.0113>, rot=<-0.0093 -0.707 0.0093 -0.707>}
 neck_01 {parent=spine_05, loc=<11.3138 0.1529 0>, rot=<0 0 -0.2003 -0.9797>}
 neck_02 {parent=neck_01, loc=<4.5627 0 0>, rot=<0 0 0.0504 -0.9987>}
 head {parent=neck_02, loc=<5.3914 0 0>, rot=<0 0 0.115 -0.9934>}
 clavicle_r {parent=spine_05, loc=<5.4831 -0.1053 3.2946>, rot=<0.7289 0.0465 -0.6829 0.009>}
 upperarm_r {parent=clavicle_r, loc=<-9.8658 1.2136 0.4105>, rot=<0.0164 -0.3245 -0.0269 -0.9454>}
 lowerarm_r {parent=upperarm_r, loc=<-26.465 0.8345 -0.259>, rot=<0.0036 0.0161 -0.2298 -0.9731>}
 hand_r {parent=lowerarm_r, loc=<-22.7355 0.3203 0.6106>, rot=<-0.6784 -0.0223 0.0735 -0.7307>}
 thumb_01_r {parent=hand_r, loc=<-1.3118 0.3191 2.459>, rot=<0.6546 -0.4042 0.0664 -0.6354>}
 thumb_02_r {parent=thumb_01_r, loc=<-5.9336 0.0448 -0.0401>, rot=<-0.0037 0.0213 0.298 -0.9543>}
 thumb_03_r {parent=thumb_02_r, loc=<-2.2223 -0.3682 0.0907>, rot=<0.0118 -0.013 0.0066 -0.9998>}
 index_metacarpal_r {parent=hand_r, loc=<-1.6535 -0.0183 2.96>, rot=<0.0158 -0.0124 -0.0173 -0.9996>}
 index_01_r {parent=index_metacarpal_r, loc=<-8.3008 -0.0357 0.0524>, rot=<-0.0044 0.0485 0.0924 -0.9945>}
 index_02_r {parent=index_01_r, loc=<-3.3587 0.3004 -0.0452>, rot=<-0.002 0.0027 0.3225 -0.9466>}
 index_03_r {parent=index_02_r, loc=<-2.5578 0.123 0.0252>, rot=<-0.002 0.0067 0.2558 -0.9667>}
 middle_metacarpal_r {parent=hand_r, loc=<-2.1856 0.1725 0.5984>, rot=<0.006 0.0168 -0.0329 -0.9993>}
 middle_01_r {parent=middle_metacarpal_r, loc=<-7.3275 0.2447 -0.0129>, rot=<-0.0208 0.0042 0.0858 -0.9961>}
 middle_02_r {parent=middle_01_r, loc=<-4.0123 -0.1401 -0.1175>, rot=<-0.0218 0.0258 0.3512 -0.9357>}
 middle_03_r {parent=middle_02_r, loc=<-2.6077 -0.0534 -0.1373>, rot=<0.0082 0.0188 0.1708 -0.9851>}
 ring_metacarpal_r {parent=hand_r, loc=<-2.2074 0.3625 -1.3638>, rot=<-0.0284 0.0722 -0.0458 -0.9959>}
 ring_01_r {parent=ring_metacarpal_r, loc=<-7.0295 0.6197 0.1446>, rot=<-0.0489 -0.0252 0.1892 -0.9804>}
 ring_02_r {parent=ring_01_r, loc=<-3.6174 -0.4647 0.1027>, rot=<0 0 0.2806 -0.9598>}
 ring_03_r {parent=ring_02_r, loc=<-2.4152 -0.114 -0.0199>, rot=<0 0 0.1189 -0.9929>}
 pinky_metacarpal_r {parent=hand_r, loc=<-1.7616 0.9753 -3.3814>, rot=<-0.0397 0.0814 -0.069 -0.9935>}
 pinky_01_r {parent=pinky_metacarpal_r, loc=<-7.1709 0.4831 0.1094>, rot=<-0.0654 -0.0035 0.2242 -0.9723>}
 pinky_02_r {parent=pinky_01_r, loc=<-2.8274 -0.0893 0.1043>, rot=<0.0037 0.0011 0.2854 -0.9584>}
 pinky_03_r {parent=pinky_02_r, loc=<-2.2046 -0.0307 0.0078>, rot=<-0.0002 0.0098 0.1445 -0.9895>}
 deform_wrist_dn_base_r {parent=hand_r, loc=<-0.0001 1.5328 0>, rot=<0 0 0 -1>}
 deform_wrist_dn_r {parent=deform_wrist_dn_base_r, loc=<0 0 0>, rot=<0 0 0 1>}
 deform_wrist_up_base_r {parent=hand_r, loc=<-0.0002 -2.3007 -0.0001>, rot=<0 0 0 -1>}
 deform_wrist_up_r {parent=deform_wrist_up_base_r, loc=<0 0 0>, rot=<0 0 0 1>}
 lowerarm_twist_02_r {parent=lowerarm_r, loc=<-10.6707 0 0>, rot=<0 0 0 1>}
 lowerarm_twist_01_r {parent=lowerarm_r, loc=<-18.316 0 0>, rot=<0 0 0 1>}
 sleeve_r {parent=lowerarm_twist_01_r, loc=<-0.0003 0.0001 0.0003>, rot=<0 0 0 1>}
 sleeves_r_02 {parent=lowerarm_r, loc=<-13.788 -1.5107 8.0716>, rot=<-0.0768 -0.7193 0.0681 -0.6871>}
 sleeves_r_01 {parent=lowerarm_r, loc=<-4.818 -3.3607 5.4593>, rot=<-0.0768 -0.7193 0.0681 -0.6871>}
 deform_elbow_in_base_r {parent=lowerarm_r, loc=<-0.1834 -3.6062 0.2572>, rot=<0 0 0 -1>}
 deform_elbow_in_r {parent=deform_elbow_in_base_r, loc=<0 0 0>, rot=<0 0 0 1>}
 deform_elbow_out_base_r {parent=lowerarm_r, loc=<0 4 -0.0001>, rot=<0 0 0 1>}
 deform_elbow_out_r {parent=deform_elbow_out_base_r, loc=<0 0 0>, rot=<0 0 0 1>}
 upperarm_twist_01_r {parent=upperarm_r, loc=<-7.2135 0.1134 0>, rot=<0 0 0 1>}
 deform_upperarm_bk_base_r {parent=upperarm_twist_01_r, loc=<6.4053 4.3746 -0.314>, rot=<0 0 0 -1>}
 deform_upperarm_bk_r {parent=deform_upperarm_bk_base_r, loc=<0 0 0>, rot=<0 0 0 1>}
 deform_upperarm_fr_base_r {parent=upperarm_twist_01_r, loc=<5.6312 -5.092 -0.3354>, rot=<0 0 0 -1>}
 deform_upperarm_fr_r {parent=deform_upperarm_fr_base_r, loc=<0 0 0>, rot=<0 0 0 1>}
 deform_upperarm_up_base_r {parent=upperarm_twist_01_r, loc=<8.891 -0.4825 -3.9517>, rot=<0 0 0 -1>}
 deform_upperarm_up_r {parent=deform_upperarm_up_base_r, loc=<0 0 0>, rot=<0 0 0 1>}
 upperarm_twist_02_r {parent=upperarm_r, loc=<-19.7998 0.1134 0>, rot=<0 0 0 1>}
 deform_upperarm_lw_r {parent=upperarm_r, loc=<-10.8352 0.0235 6.6242>, rot=<0.5271 -0.007 0.8498 -0.0028>}
 base_fur_03_r {parent=spine_05, loc=<0 0 0>, rot=<0.0284 -0.7065 -0.0284 -0.7065>}
 dyn_fur_03_r {parent=base_fur_03_r, loc=<-10.9143 12.4793 -2.9572>, rot=<-0.1245 0.1245 -0.6961 -0.6961>}
 base_fur_03_l {parent=spine_05, loc=<0 0 0>, rot=<0.7065 0.0284 -0.7065 0.0284>}
 dyn_fur_03_l {parent=base_fur_03_l, loc=<10.9143 -12.4793 2.957>, rot=<-0.1245 0.1245 -0.6961 -0.6961>}
 base_fur_01_r {parent=spine_05, loc=<13.993 -3.8023 10.1314>, rot=<0.8861 0.0356 0.0185 0.4618>}
 dyn_fur_01_r {parent=base_fur_01_r, loc=<0 0 0>, rot=<0 0 0 1>}
 base_fur_01_l {parent=spine_05, loc=<13.9927 -3.8024 -10.1314>, rot=<0.0356 -0.8861 -0.4618 0.0185>}
 dyn_fur_01_l {parent=base_fur_01_l, loc=<0 0 0>, rot=<0 0 0 -1>}
 base_fur_02_l {parent=spine_05, loc=<11.137 2.1061 -12.9612>, rot=<-0.1771 0.6846 0.6846 -0.1771>}
 dyn_fur_02_l {parent=base_fur_02_l, loc=<0 0 0>, rot=<0 0 0 1>}
 base_fur_02_r {parent=spine_05, loc=<11.1369 2.1061 12.9612>, rot=<-0.6846 -0.1771 -0.1771 -0.6846>}
 dyn_fur_02_r {parent=base_fur_02_r, loc=<0 0 0>, rot=<0 0 0 1>}
 necklace_01 {parent=spine_05, loc=<8.3146 3.7533 0>, rot=<0.2722 0.6526 0.6526 0.2722>}
 necklace_02 {parent=necklace_01, loc=<3.9472 0 0>, rot=<0 0 0 1>}
 base_fur_04_r {parent=spine_03, loc=<-1.534 12.619 8.8026>, rot=<-0.4719 -0.5266 -0.5266 -0.4719>}
 dyn_fur_04_r {parent=base_fur_04_r, loc=<0 0 0>, rot=<0 0 0 1>}
 base_fur_04_l {parent=spine_03, loc=<-1.5343 12.619 -8.8026>, rot=<-0.5266 0.4719 0.4719 -0.5266>}
 dyn_fur_04_l {parent=base_fur_04_l, loc=<0 0 0>, rot=<0 0 0 1>}
 cape_correct_02_r {parent=spine_01, loc=<5.2108 8.2018 9.7873>, rot=<0.674 0.2137 0.6806 0.1919>}
 cape_correct_02_l {parent=spine_01, loc=<5.2108 8.2018 -9.7873>, rot=<0.674 0.2137 0.6806 0.1919>}
 thigh_l {parent=pelvis, loc=<-1.0855 -0.2109 -8.7053>, rot=<0.0009 0.0355 0.0155 -0.9993>}
 calf_l {parent=thigh_l, loc=<-42.5284 -0.0019 -0.7297>, rot=<0.0006 -0.0161 -0.0391 -0.9991>}
 foot_l {parent=calf_l, loc=<-46.7401 -2.5014 -0.592>, rot=<0.0427 -0.0238 0.0353 -0.9982>}
 ball_l {parent=foot_l, loc=<-3.5658 10.8066 -0.1544>, rot=<0.0001 -0.0001 -0.7063 -0.7079>}
 calf_twist_02_l {parent=calf_l, loc=<-19.1546 -1.7943 -0.2638>, rot=<0.0237 0.0095 -0.0035 -0.9997>}
 calf_twist_01_l {parent=calf_l, loc=<-35.8483 -1.9197 -0.5794>, rot=<0.0237 0.0095 -0.0035 -0.9997>}
 deform_knee_in_base_l {parent=calf_l, loc=<0.8031 -5.0231 0.2456>, rot=<0 0 0 -1>}
 deform_knee_in_l {parent=deform_knee_in_base_l, loc=<0 0 0>, rot=<0 0 0 1>}
 deform_knee_out_base_l {parent=calf_l, loc=<0 4 0>, rot=<0 0 0 -1>}
 deform_knee_out_l {parent=deform_knee_out_base_l, loc=<0 0 0>, rot=<0 0 0 1>}
 thigh_twist_01_l {parent=thigh_l, loc=<-27.952 0 0>, rot=<0.0013 0 0 -1>}
 deform_groin_base_l {parent=thigh_l, loc=<0.4801 5.6776 -0.2486>, rot=<0 0 0 -1>}
 deform_groin_l {parent=deform_groin_base_l, loc=<0 0 0>, rot=<0 0 0 1>}
 thigh_r {parent=pelvis, loc=<-1.0855 -0.2109 8.7053>, rot=<0.0355 -0.0009 0.9993 0.0155>}
 calf_r {parent=thigh_r, loc=<42.5284 0.0019 0.7297>, rot=<0.0006 -0.0161 -0.0391 -0.9991>}
 foot_r {parent=calf_r, loc=<46.7401 2.5014 0.592>, rot=<0.0427 -0.0238 0.0353 -0.9982>}
 ball_r {parent=foot_r, loc=<3.5658 -10.8066 0.1544>, rot=<0.0001 -0.0001 -0.7063 -0.7079>}
 calf_twist_02_r {parent=calf_r, loc=<19.1546 1.7943 0.2638>, rot=<0.0237 0.0095 -0.0035 -0.9997>}
 calf_twist_01_r {parent=calf_r, loc=<35.8483 1.9197 0.5794>, rot=<0.0237 0.0095 -0.0035 -0.9997>}
 deform_knee_in_base_r {parent=calf_r, loc=<-0.8031 5.0231 -0.2457>, rot=<0 0 0 -1>}
 deform_knee_in_r {parent=deform_knee_in_base_r, loc=<0 0 0>, rot=<0 0 0 1>}
 deform_knee_out_base_r {parent=calf_r, loc=<0 -4 0>, rot=<0 0 0 -1>}
 deform_knee_out_r {parent=deform_knee_out_base_r, loc=<0 0 0>, rot=<0 0 0 1>}
 thigh_twist_01_r {parent=thigh_r, loc=<27.952 0 0>, rot=<0.0013 0 0 -1>}
 deform_groin_base_r {parent=thigh_r, loc=<-0.4801 -5.6776 0.2486>, rot=<0 0 0 -1>}
 deform_groin_r {parent=deform_groin_base_r, loc=<0 0 0>, rot=<0 0 0 1>}
 dyn_cape_b_l_base {parent=pelvis, loc=<3.4018 -10.8447 -4.2243>, rot=<0.0461 -0.7897 0.6062 -0.0823>}
 dyn_cape_b_l_01 {parent=dyn_cape_b_l_base, loc=<0 0 0>, rot=<0 0 0 1>}
 dyn_cape_b_l_02 {parent=dyn_cape_b_l_01, loc=<10.2833 0 0>, rot=<0 0 0 1>}
 dyn_cape_b_l_03 {parent=dyn_cape_b_l_02, loc=<15.2419 0 0>, rot=<0 0 0 1>}
 dyn_cape_b_l_04 {parent=dyn_cape_b_l_03, loc=<16.3545 0 0>, rot=<0 0 0 1>}
 dyn_cape_b_r_base {parent=pelvis, loc=<3.4016 -10.8447 4.2243>, rot=<0.0836 -0.6061 0.7896 -0.0471>}
 dyn_cape_b_r_01 {parent=dyn_cape_b_r_base, loc=<0 0 0>, rot=<0 0 0 1>}
 dyn_cape_b_r_02 {parent=dyn_cape_b_r_01, loc=<10.283 0 0>, rot=<0 0 0 1>}
 dyn_cape_b_r_03 {parent=dyn_cape_b_r_02, loc=<15.2418 0 0>, rot=<0 0 0 1>}
 dyn_cape_b_r_04 {parent=dyn_cape_b_r_03, loc=<16.3545 0 0>, rot=<0 0 0 1>}
 dyn_cape_s_l_base {parent=pelvis, loc=<3.7075 -6.5424 -10.2324>, rot=<0.0104 -0.9566 0.2573 -0.1366>}
 dyn_cape_s_l_01 {parent=dyn_cape_s_l_base, loc=<0 0 0>, rot=<0 0 0 1>}
 dyn_cape_s_l_02 {parent=dyn_cape_s_l_01, loc=<10.6846 0 0>, rot=<0 0 0 1>}
 dyn_cape_s_l_03 {parent=dyn_cape_s_l_02, loc=<15.5251 0 0>, rot=<0 0 0 1>}
 dyn_cape_s_l_04 {parent=dyn_cape_s_l_03, loc=<16.7207 0 0>, rot=<0 0 0 1>}
 dyn_cape_s_r_base {parent=pelvis, loc=<3.7079 -6.5424 10.2324>, rot=<0.1379 -0.2575 0.9563 -0.0092>}
 dyn_cape_s_r_01 {parent=dyn_cape_s_r_base, loc=<0 0 0>, rot=<0 0 0 1>}
 dyn_cape_s_r_02 {parent=dyn_cape_s_r_01, loc=<10.728 0 0>, rot=<0 0 0 1>}
 dyn_cape_s_r_03 {parent=dyn_cape_s_r_02, loc=<15.4359 0 0>, rot=<0 0 0 1>}
 dyn_cape_s_r_04 {parent=dyn_cape_s_r_03, loc=<16.7667 0 0>, rot=<0 0 0 1>}
 base_cape_area_l_base {parent=pelvis, loc=<-1.0855 -0.2109 -8.7053>, rot=<0.0009 0.0355 0.0155 -0.9993>}
 dyn_cape_p_l_base {parent=base_cape_area_l_base, loc=<4.6498 8.4334 1.4397>, rot=<-0.0209 -0.8007 -0.5958 -0.0592>}
 dyn_cape_p_l_01 {parent=dyn_cape_p_l_base, loc=<0 0 0>, rot=<0 0 0 1>}
 dyn_cape_p_l_02 {parent=dyn_cape_p_l_01, loc=<12.659 0 0>, rot=<0 0 0 1>}
 dyn_cape_p_l_03 {parent=dyn_cape_p_l_02, loc=<15.9001 0 0>, rot=<0 0 0 1>}
 dyn_cape_p_l_04 {parent=dyn_cape_p_l_03, loc=<17.4771 0 0>, rot=<0 0 0 1>}
 cape_correct_l_01 {parent=dyn_cape_p_l_01, loc=<-2.478 0.419 3.437>, rot=<0.6254 -0.42 0.5036 -0.423>}
 dyn_cape_f_l_base {parent=base_cape_area_l_base, loc=<5.1394 4.7913 -1.282>, rot=<0.0201 -0.9415 -0.3184 -0.1087>}
 dyn_cape_f_l_01 {parent=dyn_cape_f_l_base, loc=<0 0 0>, rot=<0 0 0 1>}
 dyn_cape_f_l_02 {parent=dyn_cape_f_l_01, loc=<12.659 0 0>, rot=<0 0 0 1>}
 dyn_cape_f_l_03 {parent=dyn_cape_f_l_02, loc=<15.9001 0 0>, rot=<0 0 0 1>}
 dyn_cape_f_l_04 {parent=dyn_cape_f_l_03, loc=<17.4771 0 0>, rot=<0 0 0 1>}
 base_cape_area_r_base {parent=pelvis, loc=<-1.0855 -0.2109 8.7053>, rot=<0.0355 -0.0009 0.9993 0.0155>}
 dyn_cape_p_r_base {parent=base_cape_area_r_base, loc=<-4.6495 -8.4334 -1.4397>, rot=<-0.0209 -0.8007 -0.5958 -0.0592>}
 dyn_cape_p_r_01 {parent=dyn_cape_p_r_base, loc=<0 0 0>, rot=<0 0 0 1>}
 dyn_cape_p_r_02 {parent=dyn_cape_p_r_01, loc=<-12.6587 0 0>, rot=<0 0 0 1>}
 dyn_cape_p_r_03 {parent=dyn_cape_p_r_02, loc=<-15.9001 0 0>, rot=<0 0 0 1>}
 dyn_cape_p_r_04 {parent=dyn_cape_p_r_03, loc=<-17.477 0 0>, rot=<0 0 0 1>}
 cape_correct_r_01 {parent=dyn_cape_p_r_01, loc=<2.478 -0.419 -3.437>, rot=<0.6254 -0.42 0.5036 -0.423>}
 dyn_cape_f_r_base {parent=base_cape_area_r_base, loc=<-5.1394 -4.7913 1.282>, rot=<-0.1327 0.1083 0.0109 -0.9852>}
 dyn_cape_f_r_01 {parent=dyn_cape_f_r_base, loc=<0 0 0>, rot=<0 0 0 1>}
 dyn_cape_f_r_02 {parent=dyn_cape_f_r_01, loc=<12.6572 0.1997 0>, rot=<0 0 0 1>}
 dyn_cape_f_r_03 {parent=dyn_cape_f_r_02, loc=<15.898 0.2508 0>, rot=<0 0 0 1>}
 dyn_cape_f_r_04 {parent=dyn_cape_f_r_03, loc=<17.4746 0.2757 0>, rot=<0 0 0 1>}
 thigh_l_null {parent=pelvis, type=Null, loc=<-1.0855 -0.2109 -8.7053>, rot=<0.0009 0.0355 0.0155 -0.9993>}
 thigh_r_null {parent=pelvis, type=Null, loc=<-1.0855 -0.2109 8.7053>, rot=<0.0355 -0.0009 0.9993 0.0155>}
 floor_joint {parent=root, loc=<0 0 -5.1328>, rot=<0 0 0 1>}
 thigh_l_ctrl {controlType=EULER_TRANSFORM, type=Control, loc=<0 0 0>, rot=<0 0 0 1>}
 calf_l_ctrl {controlType=EULER_TRANSFORM, parent=thigh_l_ctrl, type=Control, loc=<0 0 0>, rot=<0 0 0 1>}
 thigh_r_ctrl {controlType=EULER_TRANSFORM, type=Control, loc=<0 0 0>, rot=<0 0 0 1>}
 calf_r_ctrl {controlType=EULER_TRANSFORM, parent=thigh_r_ctrl, type=Control, loc=<0 0 0>, rot=<0 0 0 1>}
]
variables [
 MoveForward {type=double, default=0}
 Moving2d {type=double, default=0}
 SelectedPistol {type=double, default=0}
 Zipline {type=double, default=0}
]
