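# Physics asset for the FreshWave medium female player body, setting up secondary motion for its dyn_cape bones.
# Back, front, side and "p" groups form left/right capsule chains with tight swing and twist limits, many axes locked.
# The "p" chain bases link freely to the thighs, and a few free constraints cross-link neighboring chains.
# Leg and foot capsules provide body shapes, while a huge offset floor_joint capsule approximates the ground.
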PhysicsAsset /BRCosmetics/Characters/Player/Female/Medium/Bodies/F_MED_FreshWave/Meshes/F_MED_FreshWave_Physics
bodies {
 calf_l [{shapeType=Capsule, length=29.3669, radius=9.6331, center=<-16.0977 -2 0>, rotation=<rot 90 0 0>}]
 calf_r [{shapeType=Capsule, length=29.3669, radius=9.6331, center=<16.0977 2 0>, rotation=<rot 90 0 0>}]
 dyn_cape_b_l_01 [{
   shapeType=Capsule
   length=21.2466
   radius=6.7586
   shapeName=dyn_cape_b_l_01_capsule
   center=<1.6677 0 3.8876>
   rotation=<rot 90 153.4349 -116.5651>
  }]
 dyn_cape_b_l_02 [{shapeType=Capsule, length=18.837, radius=7, shapeName=dyn_cape_b_l_02_capsule, center=<1.9812 0 7>, rotation=<rot 90 153.4349 -116.5651>}]
 dyn_cape_b_l_03 [{
   shapeType=Capsule
   length=14.6249
   radius=8.5
   shapeName=dyn_cape_b_l_03_capsule
   center=<1.0001 0 9.08>
   rotation=<rot 90 153.4349 -116.5651>
  }]
 dyn_cape_b_l_04 [{shapeType=Capsule, length=39.808, radius=11.0387, shapeName=dyn_cape_b_l_04_capsule, center=<16.6413 0 5.54>, rotation=<rot 90 180 -90>}]
 dyn_cape_b_l_base [{shapeType=Capsule, length=0.505, radius=0.505, shapeName=dyn_cape_b_l_base_capsule, rotation=<rot 0 -90 0>}]
 dyn_cape_b_r_01 [{
   shapeType=Capsule
   length=21.2466
   radius=6.7586
   shapeName=dyn_cape_b_r_01_capsule
   center=<1.6677 0 3.8876>
   rotation=<rot 90 153.4349 -116.5651>
  }]
 dyn_cape_b_r_02 [{shapeType=Capsule, length=18.837, radius=7, shapeName=dyn_cape_b_r_02_capsule, center=<1.9812 0 7>, rotation=<rot 90 153.4349 -116.5651>}]
 dyn_cape_b_r_03 [{
   shapeType=Capsule
   length=14.6249
   radius=8.5
   shapeName=dyn_cape_b_r_03_capsule
   center=<1.0001 0 9.08>
   rotation=<rot 90 153.4349 -116.5651>
  }]
 dyn_cape_b_r_04 [{
   shapeType=Capsule
   length=39.808
   radius=11.0387
   shapeName=dyn_cape_b_r_04_capsule
   center=<16.7826 0 5.5468>
   rotation=<rot 90 153.4349 -116.5651>
  }]
 dyn_cape_b_r_base [{shapeType=Capsule, length=0.505, radius=0.505, shapeName=dyn_cape_b_r_base_capsule, rotation=<rot 0 -90 0>}]
 dyn_cape_f_l_01 [{shapeType=Capsule, length=18.2466, radius=6, shapeName=dyn_cape_f_l_01_capsule, center=<1.6677 0 6.38>, rotation=<rot 90 180 -90>}]
 dyn_cape_f_l_02 [{shapeType=Capsule, length=18.837, radius=6, shapeName=dyn_cape_f_l_02_capsule, center=<1.9812 0 7.53>, rotation=<rot 90 153.4349 -116.5651>}]
 dyn_cape_f_l_03 [{shapeType=Capsule, length=29.6249, radius=7.5, shapeName=dyn_cape_f_l_03_capsule, center=<1.0001 0 5.4>, rotation=<rot 90 180 -90>}]
 dyn_cape_f_l_04 [{shapeType=Capsule, length=20.9535, radius=6, shapeName=dyn_cape_f_l_04_capsule, center=<6.5268 0 1.2788>, rotation=<rot 90 180 -90>}]
 dyn_cape_f_l_base [{shapeType=Capsule, length=0.505, radius=0.505, shapeName=dyn_cape_f_l_base_capsule, rotation=<rot 0 -90 0>}]
 dyn_cape_f_r_01 [{shapeType=Capsule, length=21.2466, radius=5, shapeName=dyn_cape_f_r_01_capsule, center=<0 0 7.8786>, rotation=<rot 0 0 90>}]
 dyn_cape_f_r_02 [{
   shapeType=Capsule
   length=21.337
   radius=6
   shapeName=dyn_cape_f_r_02_capsule
   center=<4.5909 1.9741 8.5234>
   rotation=<rot 0 -0.0001 74.2437>
  }]
 dyn_cape_f_r_03 [{shapeType=Capsule, length=17.1249, radius=6, shapeName=dyn_cape_f_r_03_capsule, center=<3.5933 0 6.18>, rotation=<rot 0 0 90>}]
 dyn_cape_f_r_04 [{shapeType=Capsule, length=15.4535, radius=10.5, shapeName=dyn_cape_f_r_04_capsule, center=<2.1261 -0.75 6.55>, rotation=<rot 0 0 67.5625>}]
 dyn_cape_f_r_base [{shapeType=Capsule, length=0.505, radius=0.505, shapeName=dyn_cape_f_r_base_capsule, rotation=<rot 0 -90 0>}]
 dyn_cape_p_l_01 [{shapeType=Capsule, length=6, radius=5, shapeName=dyn_cape_p_l_01_capsule, center=<5.563 0 3.65>, rotation=<rot 0 90 90>}]
 dyn_cape_p_l_02 [{shapeType=Capsule, length=11, radius=7, shapeName=dyn_cape_p_l_02_capsule, center=<2.0275 -3.4 5.3>, rotation=<rot 0.0001 0 84.5312>}]
 dyn_cape_p_l_03 [{shapeType=Capsule, length=17, radius=7.5, shapeName=dyn_cape_p_l_03_capsule, center=<5.8491 -6.3 3.23>, rotation=<rot 0.0002 0 65.6391>}]
 dyn_cape_p_l_04 [{shapeType=Capsule, length=18, radius=8, shapeName=dyn_cape_p_l_04_capsule, center=<9.8538 -10 1.0991>, rotation=<rot 0.0108 0.0031 81.4654>}]
 dyn_cape_p_l_base [{shapeType=Capsule, length=0.505, radius=0.505, shapeName=dyn_cape_p_l_base_capsule, rotation=<rot 0 -90 0>}]
 dyn_cape_p_r_01 [{shapeType=Capsule, length=6, radius=5, shapeName=dyn_cape_p_l_01_capsule, center=<-4.5868 0 -4>, rotation=<rot 0 90 -90>}]
 dyn_cape_p_r_02 [{shapeType=Capsule, length=12, radius=6.5, shapeName=dyn_cape_p_l_02_capsule, center=<-0.6492 1.6 -4.4>, rotation=<rot 0 0 85.959>}]
 dyn_cape_p_r_03 [{shapeType=Capsule, length=15, radius=6.5, shapeName=dyn_cape_p_l_03_capsule, center=<-5.6837 1 -4.55>, rotation=<rot 0.0001 0 88.1945>}]
 dyn_cape_p_r_04 [{shapeType=Capsule, length=15.5, radius=8, shapeName=dyn_cape_p_l_04_capsule, center=<-9.1506 0.88 -1.4>, rotation=<rot -4.5 0 90>}]
 dyn_cape_p_r_base [{shapeType=Capsule, length=0.505, radius=0.505, shapeName=dyn_cape_p_l_base_capsule, rotation=<rot 0 -90 180>}]
 dyn_cape_s_l_01 [{shapeType=Capsule, length=21.2466, radius=3, shapeName=dyn_cape_s_l_01_capsule, center=<1.6677 0 3.88>, rotation=<rot 90 153.4349 -116.5651>}]
 dyn_cape_s_l_02 [{shapeType=Capsule, length=18.837, radius=6, shapeName=dyn_cape_s_l_02_capsule, center=<1.9812 0 8.3>, rotation=<rot 90 180 -90>}]
 dyn_cape_s_l_03 [{shapeType=Capsule, length=14.6249, radius=11.0387, shapeName=dyn_cape_s_l_03_capsule, center=<1.0001 0 10.1932>, rotation=<rot 90 180 -90>}]
 dyn_cape_s_l_04 [{
   shapeType=Capsule
   length=12.9535
   radius=11.0387
   shapeName=dyn_cape_s_l_04_capsule
   center=<3.7284 0 7.2932>
   rotation=<rot 90 153.4349 -116.5651>
  }]
 dyn_cape_s_l_base [{shapeType=Capsule, length=0.505, radius=0.505, shapeName=dyn_cape_s_l_base_capsule, rotation=<rot 0 -90 0>}]
 dyn_cape_s_r_01 [{shapeType=Capsule, length=21.2466, radius=3, shapeName=dyn_cape_s_l_01_capsule, center=<0 0 0.65>, rotation=<rot -0.0001 -0.0001 86.7714>}]
 dyn_cape_s_r_02 [{
   shapeType=Capsule
   length=18.837
   radius=6
   shapeName=dyn_cape_s_l_02_capsule
   center=<0 -3.1132 7.2693>
   rotation=<rot -0.0001 -0.0001 84.9051>
  }]
 dyn_cape_s_r_03 [{shapeType=Capsule, length=14.6249, radius=11.0387, shapeName=dyn_cape_s_l_03_capsule, center=<0 0 11>, rotation=<rot 0 0 90>}]
 dyn_cape_s_r_04 [{shapeType=Capsule, length=12.9535, radius=11.0387, shapeName=dyn_cape_s_l_04_capsule, center=<8.5999 0 6.3924>, rotation=<rot 0 0 90>}]
 dyn_cape_s_r_base [{shapeType=Capsule, length=0.505, radius=0.505, shapeName=dyn_cape_s_l_base_capsule, rotation=<rot 0 -90 180>}]
 floor_joint [{
   shapeType=Capsule
   length=0.5056
   radius=337.6725
   shapeName=floor_joint_capsule
   center=<2.6954 -23.3737 -324.4869>
   rotation=<rot 0 90 -90>
  }]
 foot_l [{shapeType=Capsule, length=15, radius=12.75, shapeName=foot_l_capsule, center=<-4.0948 6.1521 -0.1893>, rotation=<rot 0 0 90>}]
 foot_r [{shapeType=Capsule, length=15, radius=12.25, shapeName=foot_r_capsule, center=<4.0948 -4.8644 0.1893>, rotation=<rot 0 0 -90>}]
 thigh_l [
  {shapeType=Capsule, length=29.3943, radius=8, shapeName=thigh_l_capsule, center=<-15.7821 -2.0135 0.1765>, rotation=<rot 90 0 0>}
  {shapeType=Capsule, length=13, radius=7, shapeName=thigh_l_capsule, center=<4.0204 9.8466 12.0664>, rotation=<rot -5.8864 3.6391 67.524>}
  {shapeType=Capsule, length=10.5, radius=6, shapeName=thigh_l_capsule, center=<-4.9746 5.1683 10.6812>, rotation=<rot 82.421 0 0>}
 ]
 thigh_r [{shapeType=Capsule, length=29.3943, radius=8, shapeName=thigh_r_capsule, center=<15.4135 1.0948 0.028>, rotation=<rot 90 0 0>}]
}
constraints [
 {
  bone1Name=calf_l
  bone2Name=thigh_l
  swing1LimitDegrees=45
  swing1Motion=Free
  swing2LimitDegrees=45
  swing2Motion=Free
  twistLimitDegrees=45
  twistMotion=Free
 }
 {
  bone1Name=foot_l
  bone2Name=calf_l
  swing1LimitDegrees=45
  swing1Motion=Free
  swing2LimitDegrees=45
  swing2Motion=Free
  twistLimitDegrees=45
  twistMotion=Free
 }
 {
  bone1Name=calf_r
  bone2Name=thigh_r
  swing1LimitDegrees=45
  swing1Motion=Free
  swing2LimitDegrees=45
  swing2Motion=Free
  twistLimitDegrees=45
  twistMotion=Free
 }
 {
  bone1Name=foot_r
  bone2Name=calf_r
  swing1LimitDegrees=45
  swing1Motion=Free
  swing2LimitDegrees=45
  swing2Motion=Free
  twistLimitDegrees=45
  twistMotion=Free
 }
 {
  bone1Name=dyn_cape_b_l_01
  bone2Name=dyn_cape_b_l_base
  swing1LimitDegrees=2
  swing1Motion=Locked
  swing2LimitDegrees=30
  twistLimitDegrees=1
 }
 {
  bone1Name=dyn_cape_b_l_02
  bone2Name=dyn_cape_b_l_01
  swing1LimitDegrees=10
  swing1Motion=Locked
  swing2LimitDegrees=11
  twistLimitDegrees=1
  twistMotion=Locked
 }
 {
  bone1Name=dyn_cape_b_l_03
  bone2Name=dyn_cape_b_l_02
  swing1LimitDegrees=10
  swing1Motion=Locked
  swing2LimitDegrees=12
  twistLimitDegrees=1
  twistMotion=Locked
 }
 {
  bone1Name=dyn_cape_b_l_04
  bone2Name=dyn_cape_b_l_03
  swing1LimitDegrees=10
  swing1Motion=Locked
  swing2LimitDegrees=15
  twistLimitDegrees=1
  twistMotion=Locked
 }
 {
  bone1Name=dyn_cape_b_r_01
  bone2Name=dyn_cape_b_r_base
  swing1LimitDegrees=10
  swing1Motion=Locked
  swing2LimitDegrees=30
  twistLimitDegrees=1
  twistMotion=Locked
 }
 {
  bone1Name=dyn_cape_b_r_02
  bone2Name=dyn_cape_b_r_01
  swing1LimitDegrees=10
  swing1Motion=Locked
  swing2LimitDegrees=11
  twistLimitDegrees=1
  twistMotion=Locked
 }
 {
  bone1Name=dyn_cape_b_r_03
  bone2Name=dyn_cape_b_r_02
  swing1LimitDegrees=10
  swing1Motion=Locked
  swing2LimitDegrees=12
  twistLimitDegrees=1
  twistMotion=Locked
 }
 {
  bone1Name=dyn_cape_b_r_04
  bone2Name=dyn_cape_b_r_03
  swing1LimitDegrees=10
  swing1Motion=Locked
  swing2LimitDegrees=15
  twistLimitDegrees=1
  twistMotion=Locked
 }
 {
  bone1Name=dyn_cape_f_l_01
  bone2Name=dyn_cape_f_l_base
  swing1LimitDegrees=5
  swing1Motion=Locked
  swing2LimitDegrees=10
  twistLimitDegrees=1
 }
 {bone1Name=dyn_cape_f_l_02, bone2Name=dyn_cape_f_l_01, swing1LimitDegrees=15, swing1Motion=Locked, swing2LimitDegrees=15, twistLimitDegrees=1}
 {bone1Name=dyn_cape_f_l_03, bone2Name=dyn_cape_f_l_02, swing1LimitDegrees=15, swing1Motion=Locked, swing2LimitDegrees=20, twistLimitDegrees=1}
 {bone1Name=dyn_cape_f_l_04, bone2Name=dyn_cape_f_l_03, swing1LimitDegrees=15, swing1Motion=Locked, swing2LimitDegrees=15, twistLimitDegrees=1}
 {
  bone1Name=dyn_cape_f_r_01
  bone2Name=dyn_cape_f_r_base
  swing1LimitDegrees=15
  swing1Motion=Locked
  swing2LimitDegrees=10
  twistLimitDegrees=1
 }
 {bone1Name=dyn_cape_f_r_02, bone2Name=dyn_cape_f_r_01, swing1LimitDegrees=15, swing2LimitDegrees=15, twistLimitDegrees=1}
 {bone1Name=dyn_cape_f_r_03, bone2Name=dyn_cape_f_r_02, swing1LimitDegrees=15, swing1Motion=Locked, swing2LimitDegrees=20, twistLimitDegrees=1}
 {bone1Name=dyn_cape_f_r_04, bone2Name=dyn_cape_f_r_03, swing1LimitDegrees=15, swing1Motion=Locked, swing2LimitDegrees=15, twistLimitDegrees=1}
 {
  bone1Name=dyn_cape_s_l_01
  bone2Name=dyn_cape_s_l_base
  swing1LimitDegrees=10
  swing1Motion=Locked
  swing2LimitDegrees=10
  twistLimitDegrees=1
 }
 {bone1Name=dyn_cape_s_l_02, bone2Name=dyn_cape_s_l_01, swing1LimitDegrees=10, swing1Motion=Locked, swing2LimitDegrees=10, twistLimitDegrees=1}
 {bone1Name=dyn_cape_s_l_03, bone2Name=dyn_cape_s_l_02, swing1LimitDegrees=10, swing1Motion=Locked, swing2LimitDegrees=20, twistLimitDegrees=1}
 {bone1Name=dyn_cape_s_l_04, bone2Name=dyn_cape_s_l_03, swing1LimitDegrees=10, swing1Motion=Locked, swing2LimitDegrees=15, twistLimitDegrees=1}
 {
  bone1Name=dyn_cape_b_l_04
  bone2Name=dyn_cape_s_l_04
  swing1LimitDegrees=45
  swing1Motion=Free
  swing2LimitDegrees=45
  swing2Motion=Free
  twistLimitDegrees=45
  twistMotion=Free
 }
 {
  bone1Name=dyn_cape_b_l_03
  bone2Name=dyn_cape_b_r_03
  swing1LimitDegrees=45
  swing1Motion=Free
  swing2LimitDegrees=45
  swing2Motion=Free
  twistLimitDegrees=45
  twistMotion=Free
 }
 {
  bone1Name=dyn_cape_s_l_04
  bone2Name=dyn_cape_f_l_04
  swing1LimitDegrees=45
  swing1Motion=Free
  swing2LimitDegrees=45
  swing2Motion=Free
  twistLimitDegrees=45
  twistMotion=Free
 }
 {bone1Name=dyn_cape_p_l_04, bone2Name=dyn_cape_p_l_03, swing1LimitDegrees=10, swing2LimitDegrees=20, twistLimitDegrees=45, twistMotion=Locked}
 {bone1Name=dyn_cape_p_l_03, bone2Name=dyn_cape_p_l_02, swing1LimitDegrees=10, swing2LimitDegrees=12, twistLimitDegrees=45, twistMotion=Locked}
 {bone1Name=dyn_cape_p_l_02, bone2Name=dyn_cape_p_l_01, swing1LimitDegrees=3, swing2LimitDegrees=12, twistLimitDegrees=45, twistMotion=Locked}
 {bone1Name=dyn_cape_p_l_01, bone2Name=dyn_cape_p_l_base, swing1LimitDegrees=5, swing2LimitDegrees=7, twistLimitDegrees=45, twistMotion=Locked}
 {
  bone1Name=dyn_cape_p_l_base
  bone2Name=thigh_l
  swing1LimitDegrees=45
  swing1Motion=Free
  swing2LimitDegrees=45
  swing2Motion=Free
  twistLimitDegrees=45
  twistMotion=Free
 }
 {
  bone1Name=dyn_cape_s_r_01
  bone2Name=dyn_cape_s_r_base
  swing1LimitDegrees=10
  swing1Motion=Locked
  swing2LimitDegrees=10
  twistLimitDegrees=1
  twistMotion=Locked
 }
 {
  bone1Name=dyn_cape_s_r_02
  bone2Name=dyn_cape_s_r_01
  swing1LimitDegrees=10
  swing1Motion=Locked
  swing2LimitDegrees=10
  twistLimitDegrees=1
  twistMotion=Locked
 }
 {
  bone1Name=dyn_cape_s_r_03
  bone2Name=dyn_cape_s_r_02
  swing1LimitDegrees=10
  swing1Motion=Locked
  swing2LimitDegrees=20
  twistLimitDegrees=1
  twistMotion=Locked
 }
 {
  bone1Name=dyn_cape_s_r_04
  bone2Name=dyn_cape_s_r_03
  swing1LimitDegrees=10
  swing1Motion=Locked
  swing2LimitDegrees=15
  twistLimitDegrees=1
  twistMotion=Locked
 }
 {bone1Name=dyn_cape_p_r_04, bone2Name=dyn_cape_p_r_03, swing1LimitDegrees=10, swing2LimitDegrees=10, twistLimitDegrees=45, twistMotion=Locked}
 {bone1Name=dyn_cape_p_r_03, bone2Name=dyn_cape_p_r_02, swing1LimitDegrees=10, swing2LimitDegrees=10, twistLimitDegrees=45, twistMotion=Locked}
 {bone1Name=dyn_cape_p_r_02, bone2Name=dyn_cape_p_r_01, swing1LimitDegrees=5, swing2LimitDegrees=12, twistLimitDegrees=45, twistMotion=Locked}
 {
  bone1Name=dyn_cape_p_r_01
  bone2Name=dyn_cape_p_r_base
  swing1LimitDegrees=5
  swing2LimitDegrees=12
  twistLimitDegrees=45
  twistMotion=Locked
 }
 {
  bone1Name=dyn_cape_p_r_base
  bone2Name=thigh_r
  swing1LimitDegrees=45
  swing1Motion=Free
  swing2LimitDegrees=45
  swing2Motion=Free
  twistLimitDegrees=45
  twistMotion=Free
 }
]
