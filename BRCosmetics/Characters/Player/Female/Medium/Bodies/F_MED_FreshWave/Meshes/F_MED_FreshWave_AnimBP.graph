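# Post-process-style AnimBlueprint for the F_MED_FreshWave cosmetic body, adding secondary motion to its coat/cape, sleeves, fur and necklace.
# The parent-mesh pose passes through Control Rigs and corrective bone constraints and transforms into three RigidBody simulations, selected by crouch and front-end skydiving state.
# The cached result then receives AnimDynamics and final emote tweaks.
# EventGraph derives alphas and sim-space settings from movement, zipline, weapon and player states.
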AnimBlueprint /BRCosmetics/Characters/Player/Female/Medium/Bodies/F_MED_FreshWave/Meshes/F_MED_FreshWave_AnimBP
target_skeleton=&/F_MED_FreshWave_Skeleton
graphs [
 AnimGraph {nodes [
   n1 = Anim Dynamics - Bone: sleeves_l_01(ComponentPose=Local To Component(LocalPose=Use cached pose 'RigidBody Cache'()), Alpha=1, ExternalForce="0,0,0") {
    angularDampingOverride=1
    angularSpringConstant=10000
    bAngularSpring=true
    bOverrideAngularDamping=true
    bUseGravityOverride=true
    gravityScale=0.35
    simulationSpace=RootRelative
    alphaBoolBlend {blendOption=Linear}
    boundBone {boneName=sleeves_l_01}
    componentLinearAccScale=<0.7 0.7 0.7>
    componentLinearVelScale=<0.7 0.7 0.7>
    gravityOverride {z=-980}
    physicsBodyDefinitions [{
      boundBone {boneName=sleeves_l_01}
      boxExtents=<10 20 20>
      constraintSetup {angularLimitsMax=<0 15 55>, angularLimitsMin=<0 -15 -55>, angularTarget=<1 0 0>}
      localJointOffset=<5 0 0>
     }]
   }
   n2 = Anim Dynamics - Bone: sleeves_l_02(ComponentPose=@n1, Alpha=1, ExternalForce="0,0,0") {
    angularDampingOverride=1
    angularSpringConstant=10000
    bAngularSpring=true
    bOverrideAngularDamping=true
    bUseGravityOverride=true
    gravityScale=0.35
    simulationSpace=RootRelative
    alphaBoolBlend {blendOption=Linear}
    boundBone {boneName=sleeves_l_02}
    componentLinearAccScale=<0.7 0.7 0.7>
    componentLinearVelScale=<0.7 0.7 0.7>
    gravityOverride {z=-980}
    physicsBodyDefinitions [{
      boundBone {boneName=sleeves_l_02}
      boxExtents=<10 20 20>
      constraintSetup {angularLimitsMax=<0 15 55>, angularLimitsMin=<0 -15 -55>, angularTarget=<1 0 0>}
      localJointOffset=<5 0 0>
     }]
   }
   n3 = Anim Dynamics - Bone: sleeves_r_01(ComponentPose=@n2, Alpha=1, ExternalForce="0,0,0") {
    angularDampingOverride=1
    angularSpringConstant=10000
    bAngularSpring=true
    bOverrideAngularDamping=true
    bUseGravityOverride=true
    gravityScale=0.35
    simulationSpace=RootRelative
    alphaBoolBlend {blendOption=Linear}
    boundBone {boneName=sleeves_r_01}
    componentLinearAccScale=<0.7 0.7 0.7>
    componentLinearVelScale=<0.7 0.7 0.7>
    gravityOverride {z=-980}
    physicsBodyDefinitions [{
      boundBone {boneName=sleeves_r_01}
      boxExtents=<10 20 20>
      constraintSetup {angularLimitsMax=<0 15 55>, angularLimitsMin=<0 -15 -55>, angularTarget=<1 0 0>}
      localJointOffset=<5 0 0>
     }]
   }
   n4 = Anim Dynamics - Bone: sleeves_r_02(ComponentPose=@n3, Alpha=1, ExternalForce="0,0,0") {
    angularDampingOverride=1
    angularSpringConstant=10000
    bAngularSpring=true
    bOverrideAngularDamping=true
    bUseGravityOverride=true
    gravityScale=0.35
    simulationSpace=RootRelative
    alphaBoolBlend {blendOption=Linear}
    boundBone {boneName=sleeves_r_02}
    componentLinearAccScale=<0.7 0.7 0.7>
    componentLinearVelScale=<0.7 0.7 0.7>
    gravityOverride {z=-980}
    physicsBodyDefinitions [{
      boundBone {boneName=sleeves_r_02}
      boxExtents=<10 20 20>
      constraintSetup {angularLimitsMax=<0 15 55>, angularLimitsMin=<0 -15 -55>, angularTarget=<1 0 0>}
      localJointOffset=<5 0 0>
     }]
   }
   n5 = Anim Dynamics - Bone: dyn_fur_01_l(ComponentPose=@n4, Alpha=1, ExternalForce="0,0,0") {
    angularDampingOverride=1
    angularSpringConstant=20000
    bAngularSpring=true
    bOverrideAngularDamping=true
    bUseGravityOverride=true
    gravityScale=0.35
    simulationSpace=RootRelative
    alphaBoolBlend {blendOption=Linear}
    boundBone {boneName=dyn_fur_01_l}
    componentLinearAccScale=<0.4 0.4 0.4>
    componentLinearVelScale=<0.4 0.4 0.4>
    gravityOverride {z=-980}
    physicsBodyDefinitions [{
      boundBone {boneName=dyn_fur_01_l}
      boxExtents=<20 30 30>
      constraintSetup {angularLimitsMax=<0 0 15>, angularLimitsMin=<0 0 -15>, angularTarget=<1 0 0>}
      localJointOffset=<4 0 0>
     }]
   }
   n6 = Anim Dynamics - Bone: dyn_fur_01_r(ComponentPose=@n5, Alpha=1, ExternalForce="0,0,0") {
    angularDampingOverride=1
    angularSpringConstant=20000
    bAngularSpring=true
    bOverrideAngularDamping=true
    bUseGravityOverride=true
    gravityScale=0.35
    simulationSpace=RootRelative
    alphaBoolBlend {blendOption=Linear}
    boundBone {boneName=dyn_fur_01_r}
    componentLinearAccScale=<0.4 0.4 0.4>
    componentLinearVelScale=<0.4 0.4 0.4>
    gravityOverride {z=-980}
    physicsBodyDefinitions [{
      boundBone {boneName=dyn_fur_01_r}
      boxExtents=<20 30 30>
      constraintSetup {angularLimitsMax=<0 0 15>, angularLimitsMin=<0 0 -15>, angularTarget=<1 0 0>}
      localJointOffset=<4 0 0>
     }]
   }
   n7 = Anim Dynamics - Bone: dyn_fur_02_l(ComponentPose=@n6, Alpha=1, ExternalForce="0,0,0") {
    angularDampingOverride=1
    angularSpringConstant=20000
    bAngularSpring=true
    bOverrideAngularDamping=true
    bUseGravityOverride=true
    gravityScale=0.35
    simulationSpace=RootRelative
    alphaBoolBlend {blendOption=Linear}
    boundBone {boneName=dyn_fur_02_l}
    componentLinearAccScale=<0.4 0.4 0.4>
    componentLinearVelScale=<0.4 0.4 0.4>
    gravityOverride {z=-980}
    physicsBodyDefinitions [{
      boundBone {boneName=dyn_fur_02_l}
      boxExtents=<20 30 30>
      constraintSetup {angularLimitsMax=<0 0 15>, angularLimitsMin=<0 0 -15>, angularTarget=<1 0 0>}
      localJointOffset=<4 0 0>
     }]
   }
   n8 = Anim Dynamics - Bone: dyn_fur_02_r(ComponentPose=@n7, Alpha=1, ExternalForce="0,0,0") {
    angularDampingOverride=1
    angularSpringConstant=20000
    bAngularSpring=true
    bOverrideAngularDamping=true
    bUseGravityOverride=true
    gravityScale=0.35
    simulationSpace=RootRelative
    alphaBoolBlend {blendOption=Linear}
    boundBone {boneName=dyn_fur_02_r}
    componentLinearAccScale=<0.4 0.4 0.4>
    componentLinearVelScale=<0.4 0.4 0.4>
    gravityOverride {z=-980}
    physicsBodyDefinitions [{
      boundBone {boneName=dyn_fur_02_r}
      boxExtents=<20 30 30>
      constraintSetup {angularLimitsMax=<0 0 15>, angularLimitsMin=<0 0 -15>, angularTarget=<1 0 0>}
      localJointOffset=<4 0 0>
     }]
   }
   n9 = Anim Dynamics - Bone: dyn_fur_03_l(ComponentPose=@n8, Alpha=1, ExternalForce="0,0,0") {
    angularDampingOverride=1
    angularSpringConstant=20000
    bAngularSpring=true
    bOverrideAngularDamping=true
    bUseGravityOverride=true
    gravityScale=0.35
    simulationSpace=RootRelative
    alphaBoolBlend {blendOption=Linear}
    boundBone {boneName=dyn_fur_03_l}
    componentLinearAccScale=<0.4 0.4 0.4>
    componentLinearVelScale=<0.4 0.4 0.4>
    gravityOverride {z=-980}
    physicsBodyDefinitions [{
      boundBone {boneName=dyn_fur_03_l}
      boxExtents=<20 30 30>
      constraintSetup {angularLimitsMax=<0 0 15>, angularLimitsMin=<0 0 -15>, angularTarget=<1 0 0>}
      localJointOffset=<4 0 0>
     }]
   }
   n10 = Anim Dynamics - Bone: dyn_fur_03_r(ComponentPose=@n9, Alpha=1, ExternalForce="0,0,0") {
    angularDampingOverride=1
    angularSpringConstant=20000
    bAngularSpring=true
    bOverrideAngularDamping=true
    bUseGravityOverride=true
    gravityScale=0.35
    simulationSpace=RootRelative
    alphaBoolBlend {blendOption=Linear}
    boundBone {boneName=dyn_fur_03_r}
    componentLinearAccScale=<0.4 0.4 0.4>
    componentLinearVelScale=<0.4 0.4 0.4>
    gravityOverride {z=-980}
    physicsBodyDefinitions [{
      boundBone {boneName=dyn_fur_03_r}
      boxExtents=<20 30 30>
      constraintSetup {angularLimitsMax=<0 0 15>, angularLimitsMin=<0 0 -15>, angularTarget=<1 0 0>}
      localJointOffset=<4 0 0>
     }]
   }
   n11 = Get Alpha_isFrontEnd / BR Necklace() -> (Alpha_isFrontEnd / BR Necklace:double)
   n12 = Anim Dynamics - Bone: necklace_01(ComponentPose=@n10, Alpha=@n11, ExternalForce="0,0,0") {
    angularDampingOverride=1
    angularSpringConstant=20000
    bAngularSpring=true
    bOverrideAngularDamping=true
    bUseGravityOverride=true
    gravityScale=0.35
    simulationSpace=World
    alphaBoolBlend {blendOption=Linear}
    boundBone {boneName=necklace_01}
    componentLinearAccScale=<0.7 0.7 0.7>
    componentLinearVelScale=<0.7 0.7 0.7>
    gravityOverride {z=-980}
    physicsBodyDefinitions [{
      boundBone {boneName=necklace_01}
      boxExtents=<20 30 30>
      constraintSetup {angularLimitsMax=<5 0 15>, angularLimitsMin=<-5 0 -15>, angularTarget=<1 0 0>}
      localJointOffset=<4 0 0>
     }]
   }
   n13 = Anim Dynamics - Bone: necklace_02(ComponentPose=@n12, Alpha=@n11, ExternalForce="0,0,0") {
    angularDampingOverride=1
    angularSpringConstant=20000
    bAngularSpring=true
    bOverrideAngularDamping=true
    bUseGravityOverride=true
    gravityScale=0.35
    simulationSpace=World
    alphaBoolBlend {blendOption=Linear}
    boundBone {boneName=necklace_02}
    componentLinearAccScale=<0.7 0.7 0.7>
    componentLinearVelScale=<0.7 0.7 0.7>
    gravityOverride {z=-980}
    physicsBodyDefinitions [{
      boundBone {boneName=necklace_02}
      boxExtents=<20 30 30>
      constraintSetup {angularLimitsMax=<5 0 15>, angularLimitsMin=<-5 0 -15>, angularTarget=<1 0 0>}
      localJointOffset=<4 0 0>
     }]
   }
   n14 = "Transform (Modify) Bone - Bone: dyn_cape_f_l_base"!(ComponentPose=@n13, bAlphaBoolEnabled:bool=True, Translation="-1.285,2.498,-0.178", Rotation="11.885,35.493,47.737", Scale="1,1,1") -> (Pose) {
    alphaInputType=Bool
    rotationMode=BMM_Additive
    rotationSpace=BCS_BoneSpace
    translationMode=BMM_Additive
    translationSpace=BCS_BoneSpace
    alphaBoolBlend {blendInTime=0.25, blendOption=CircularOut, blendOutTime=1}
    boneToModify {boneName=dyn_cape_f_l_base}
    rotation=<rot 11.885 35.493 47.7369>
    translation=<-1.285 2.498 -0.178>
   }
   n15 = "Transform (Modify) Bone - Bone: dyn_cape_p_l_base"!(ComponentPose=@n14, bAlphaBoolEnabled:bool=True, Translation="1.117,1.389,-5.419", Rotation="36.705,17.676,71.243", Scale="1,1,1") -> (Pose) {
    alphaInputType=Bool
    rotationMode=BMM_Additive
    rotationSpace=BCS_BoneSpace
    translationMode=BMM_Additive
    translationSpace=BCS_BoneSpace
    alphaBoolBlend {blendInTime=0.25, blendOption=CircularOut, blendOutTime=1}
    boneToModify {boneName=dyn_cape_p_l_base}
    rotation=<rot 36.705 17.676 71.243>
    translation=<1.117 1.389 -5.419>
   }
   n16 = "Transform (Modify) Bone - Bone: cape_correct_02_l"!(ComponentPose=@n15, bAlphaBoolEnabled:bool=True, Translation="-7.072,-4.537,0.176", Rotation="-66.294,91.517,18.438", Scale="1,1,1") -> (Pose) {
    alphaInputType=Bool
    rotationMode=BMM_Additive
    rotationSpace=BCS_BoneSpace
    translationMode=BMM_Additive
    translationSpace=BCS_BoneSpace
    alphaBoolBlend {blendInTime=1, blendOption=QuarticInOut}
    boneToModify {boneName=cape_correct_02_l}
    rotation=<rot -66.294 91.517 18.438>
    translation=<-7.072 -4.537 0.176>
   }
   n17 = Get PlayingBS() -> (PlayingBS:double)
   n18 = "Transform (Modify) Bone - Bone: dyn_cape_b_r_02"(ComponentPose=@n16, Alpha=@n17, Translation="0,4,4", Rotation="20,0,0", Scale="1,1,1") {
    rotationMode=BMM_Additive
    rotationSpace=BCS_BoneSpace
    translationMode=BMM_Additive
    translationSpace=BCS_BoneSpace
    alphaBoolBlend {blendOption=Linear}
    boneToModify {boneName=dyn_cape_b_r_02}
    rotation {pitch=20}
    translation {y=4, z=4}
   }
   n19 = "Transform (Modify) Bone - Bone: dyn_cape_s_r_02"(ComponentPose=@n18, Alpha=@n17, Translation="2.665,2.954,5.887", Rotation="38.540,22.822,-28.997", Scale="1,1,1") {
    rotationMode=BMM_Additive
    rotationSpace=BCS_BoneSpace
    translationMode=BMM_Additive
    translationSpace=BCS_BoneSpace
    alphaBoolBlend {blendOption=Linear}
    boneToModify {boneName=dyn_cape_s_r_02}
    rotation=<rot 38.54 22.822 -28.997>
    translation=<2.665 2.954 5.887>
   }
   Output Pose(Result=Component To Local(ComponentPose=@n19)) {layerGroup=None}
   n20 = Control Rig!(
    Source=Use cached pose 'Input_CtrlRig_Cache'()
    Alpha
    SelectedPistol:double=0
    Moving2d:double=Get Alpha_Moving2d() -> (Alpha_Moving2d:double)
    MoveForward:double=Get Alpha_is_Forward() -> (Alpha_is_Forward:double)
    Zipline:double=Get OnZipline_Float() -> (OnZipline_Float:double)
   ) -> (Pose) {
    alphaBoolBlend {blendOption=Linear}
    controlRigAssetReference {blueprintRigClass=&/F_MED_FreshWave_CoatCtrlRig.F_MED_FreshWave_CoatCtrlRig_C}
   }
   Save cached pose 'Coat_CtrlRig_Cache'(Pose=@n20) {cacheName=Coat_CtrlRig_Cache}
   n21 = "Constraint - base_cape_area_l_base (P)"(ComponentPose=Local To Component(LocalPose=Use cached pose 'Input_CtrlRig_Cache'()), Alpha=1, ConstraintWeights_0=1) {
    alphaBoolBlend {blendOption=Linear}
    boneToModify {boneName=base_cape_area_l_base}
    constraintSetup [{targetBone {boneName=thigh_l}}]
    constraintWeights [1]
   }
   n22 = Look At - Bone: dyn_cape_s_r_base(ComponentPose=@n21, Alpha=1) {
    alphaBoolBlend {blendOption=Linear}
    alphaScaleBiasClamp {bInterpResult=true}
    boneToModify {boneName=dyn_cape_s_r_base}
    lookAtLocation=<23.8121 -32.5194 46.0129>
    lookAtTarget {boneReference {boneName=foot_r}}
    lookAt_Axis {axis=<1 0>}
   }
   n23 = Look At - Bone: dyn_cape_b_r_base(ComponentPose=@n22, Alpha=1) {
    alphaBoolBlend {blendOption=Linear}
    alphaScaleBiasClamp {bInterpResult=true}
    boneToModify {boneName=dyn_cape_b_r_base}
    lookAtLocation=<43.4466 -57.9106 37.9542>
    lookAtTarget {boneReference {boneName=foot_r}}
    lookAt_Axis {axis=<1 0>}
   }
   n24 = Look At - Bone: dyn_cape_s_l_base(ComponentPose=@n23, Alpha=1) {
    alphaBoolBlend {blendOption=Linear}
    alphaScaleBiasClamp {bInterpResult=true}
    boneToModify {boneName=dyn_cape_s_l_base}
    lookAtLocation=<18.3316 40.5419 -48.5112>
    lookAtTarget {boneReference {boneName=foot_l}}
    lookAt_Axis {axis=<1 0>}
   }
   n25 = Look At - Bone: dyn_cape_b_l_base(ComponentPose=@n24, Alpha=1) {
    alphaBoolBlend {blendOption=Linear}
    alphaScaleBiasClamp {bInterpResult=true}
    boneToModify {boneName=dyn_cape_b_l_base}
    lookAtLocation=<-45.3018 42.7025 -2.9804>
    lookAtTarget {boneReference {boneName=foot_l}}
    lookAt_Axis {axis=<1 0>}
   }
   n26 = "Transform (Modify) Bone - Bone: base_cape_area_r_base"!(ComponentPose=@n25, Alpha=1, Translation="-2.393,0.769,11.498", Rotation="47.568,-8.157,28.538") -> (Pose) {
    rotationMode=BMM_Additive
    rotationSpace=BCS_BoneSpace
    translationMode=BMM_Additive
    translationSpace=BCS_BoneSpace
    alphaBoolBlend {blendOption=Linear}
    alphaScaleBiasClamp {bInterpResult=true, interpSpeedDecreasing=1}
    boneToModify {boneName=base_cape_area_r_base}
    rotation=<rot 47.568 -8.157 28.538>
    translation=<-2.3932 0.7692 11.4982>
   }
   n27 = "Transform (Modify) Bone - Bone: dyn_cape_s_r_base"!(ComponentPose=@n26, Alpha=1, Translation="4.190,1.001,0.197", Rotation="0,0,0") -> (Pose) {
    rotationMode=BMM_Additive
    rotationSpace=BCS_BoneSpace
    translationMode=BMM_Additive
    translationSpace=BCS_BoneSpace
    alphaBoolBlend {blendOption=Linear}
    alphaScaleBiasClamp {bInterpResult=true, interpSpeedDecreasing=1}
    boneToModify {boneName=dyn_cape_s_r_base}
    translation=<4.1904 1.0011 0.197>
   }
   n28 = "Transform (Modify) Bone - Bone: base_cape_area_l_base"!(ComponentPose=@n27, Alpha=1, Translation="-0.211,-1.369,-4.309", Rotation="22.657,-0.000,0.000") -> (Pose) {
    rotationMode=BMM_Additive
    rotationSpace=BCS_BoneSpace
    translationMode=BMM_Additive
    translationSpace=BCS_BoneSpace
    alphaBoolBlend {blendOption=Linear}
    alphaScaleBiasClamp {bInterpResult=true, interpSpeedDecreasing=1}
    boneToModify {boneName=base_cape_area_l_base}
    rotation {pitch=22.657, yaw=0}
    translation=<-0.211 -1.369 -4.309>
   }
   n29 = "Constraint - base_cape_area_l_base (P)"(ComponentPose=Local To Component(LocalPose=Use cached pose 'Input_CtrlRig_Cache'()), Alpha=0.5, ConstraintWeights_0=1) {
    alphaBoolBlend {blendOption=Linear}
    boneToModify {boneName=base_cape_area_l_base}
    constraintSetup [{targetBone {boneName=thigh_l}}]
    constraintWeights [1]
   }
   n30 = "Constraint - base_cape_area_r_base (P)"(ComponentPose=@n29, Alpha=1, ConstraintWeights_0=1) {
    alphaBoolBlend {blendOption=Linear}
    boneToModify {boneName=base_cape_area_r_base}
    constraintSetup [{targetBone {boneName=thigh_r}}]
    constraintWeights [1]
   }
   n31 = "Transform (Modify) Bone - Bone: base_cape_area_l_base"(ComponentPose=@n30, Alpha=1, Translation="-7.484,2.866,-2.901", Rotation="11.082,-2.483,21.810") {
    rotationMode=BMM_Additive
    rotationSpace=BCS_BoneSpace
    translationMode=BMM_Additive
    translationSpace=BCS_BoneSpace
    alphaBoolBlend {blendOption=Linear}
    boneToModify {boneName=base_cape_area_l_base}
    rotation=<rot 11.082 -2.483 21.81>
    translation=<-7.4844 2.866 -2.9011>
   }
   n32 = "Transform (Modify) Bone - Bone: base_cape_area_r_base"(ComponentPose=@n31, Alpha=1, Translation="16.195,-14.700,-0.228", Rotation="5.357,-45.240,-117.571") {
    rotationMode=BMM_Additive
    rotationSpace=BCS_BoneSpace
    translationMode=BMM_Additive
    translationSpace=BCS_BoneSpace
    alphaBoolBlend {blendOption=Linear}
    alphaScaleBiasClamp {interpSpeedDecreasing=3}
    boneToModify {boneName=base_cape_area_r_base}
    rotation=<rot 5.357 -45.24 -117.571>
    translation=<16.195 -14.6998 -0.2277>
   }
   n33 = "Transform (Modify) Bone - Bone: dyn_cape_f_r_base"!(ComponentPose=@n32, Alpha=1, Translation="2.102,3.179,3.286", Rotation="0,0,0") -> (Pose) {
    rotationMode=BMM_Additive
    rotationSpace=BCS_BoneSpace
    translationMode=BMM_Additive
    translationSpace=BCS_BoneSpace
    alphaBoolBlend {blendOption=Linear}
    boneToModify {boneName=dyn_cape_f_r_base}
    translation=<2.102 3.179 3.286>
   }
   n34 = "Transform (Modify) Bone - Bone: dyn_cape_b_r_base"!(ComponentPose=@n33, Alpha=1, Translation="-1.997,0.000,8.169", Rotation="29.992,-180.000,-180.000") -> (Pose) {
    rotationMode=BMM_Additive
    rotationSpace=BCS_BoneSpace
    translationMode=BMM_Additive
    translationSpace=BCS_BoneSpace
    alphaBoolBlend {blendOption=Linear}
    boneToModify {boneName=dyn_cape_b_r_base}
    rotation=<rot 29.992 -180 -180>
    translation {x=-1.997, z=8.169}
   }
   n35 = "Transform (Modify) Bone - Bone: dyn_cape_b_l_base"!(ComponentPose=@n34, Alpha=1, Translation="-2.144,0.000,1.477", Rotation="55.443,0.000,-0.000") -> (Pose) {
    rotationMode=BMM_Additive
    rotationSpace=BCS_BoneSpace
    translationMode=BMM_Additive
    translationSpace=BCS_BoneSpace
    alphaBoolBlend {blendOption=Linear}
    boneToModify {boneName=dyn_cape_b_l_base}
    rotation=<rot 55.443 0 0>
    translation {x=-2.144, z=1.477}
   }
   n36 = "Transform (Modify) Bone - Bone: dyn_cape_s_r_base"(ComponentPose=@n35, Alpha=1, Translation="0,0,4.797", Rotation="53.991,-180.000,-180.000") {
    rotationMode=BMM_Additive
    rotationSpace=BCS_BoneSpace
    translationMode=BMM_Additive
    translationSpace=BCS_BoneSpace
    alphaBoolBlend {blendOption=Linear}
    boneToModify {boneName=dyn_cape_s_r_base}
    rotation=<rot 53.991 -180 -180>
    translation {z=4.797}
   }
   n37 = "Transform (Modify) Bone - Bone: cape_correct_r_01"!(ComponentPose=@n36, Alpha=1, Translation="-4.429,2.403,-5.636", Rotation="15.943,-6.434,12.507") -> (Pose) {
    rotationMode=BMM_Additive
    rotationSpace=BCS_BoneSpace
    translationMode=BMM_Additive
    translationSpace=BCS_BoneSpace
    alphaBoolBlend {blendOption=Linear}
    boneToModify {boneName=cape_correct_r_01}
    rotation=<rot 15.943 -6.434 12.507>
    translation=<-4.429 2.403 -5.636>
   }
   n38 = Blend Poses by bool(BlendPose_0=Component To Local(ComponentPose=@n28), BlendPose_1=Component To Local(ComponentPose=@n37), BlendTime_0=0.5, BlendTime_1=0.5) {blendType=Linear}
   n39 = "Constraint - base_cape_area_l_base (P)"(ComponentPose=Local To Component(LocalPose=Use cached pose 'Input_CtrlRig_Cache'()), Alpha=0.6, ConstraintWeights_0=1) {
    alphaBoolBlend {blendOption=Linear}
    boneToModify {boneName=base_cape_area_l_base}
    constraintSetup [{targetBone {boneName=thigh_l}}]
    constraintWeights [1]
   }
   n40 = "Constraint - base_cape_area_r_base (P)"(ComponentPose=@n39, Alpha=1, ConstraintWeights_0=1) {
    alphaBoolBlend {blendOption=Linear}
    boneToModify {boneName=base_cape_area_r_base}
    constraintSetup [{targetBone {boneName=thigh_r}}]
    constraintWeights [1]
   }
   n41 = "Transform (Modify) Bone - Bone: base_cape_area_l_base"(ComponentPose=@n40, Alpha=1, Translation="-2.057,3.817,-1.729", Rotation="27.665,12.257,15.544") {
    rotationMode=BMM_Additive
    rotationSpace=BCS_BoneSpace
    translationMode=BMM_Additive
    translationSpace=BCS_BoneSpace
    alphaBoolBlend {blendInTime=0.2, blendOption=Linear, blendOutTime=0.35}
    alphaScaleBiasClamp {bInterpResult=true, interpSpeedDecreasing=3}
    boneToModify {boneName=base_cape_area_l_base}
    rotation=<rot 27.665 12.257 15.544>
    translation=<-2.057 3.817 -1.729>
   }
   n42 = "Transform (Modify) Bone - Bone: base_cape_area_r_base"(ComponentPose=@n41, Alpha=1, Translation="-0.000,0.927,3.535", Rotation="-0.000,-0.000,14.715") {
    rotationMode=BMM_Additive
    rotationSpace=BCS_BoneSpace
    translationMode=BMM_Additive
    translationSpace=BCS_BoneSpace
    alphaBoolBlend {blendInTime=0.2, blendOption=Linear, blendOutTime=0.35}
    alphaScaleBiasClamp {bInterpResult=true}
    boneToModify {boneName=base_cape_area_r_base}
    rotation {roll=14.715}
    translation {y=0.927, z=3.535}
   }
   n43 = "Transform (Modify) Bone - Bone: cape_correct_l_01"!(ComponentPose=@n42, Alpha=1, Translation="0.316,0.906,5.499", Rotation="3.136,-0.572,12.563") -> (Pose) {
    rotationMode=BMM_Additive
    rotationSpace=BCS_BoneSpace
    translationMode=BMM_Additive
    translationSpace=BCS_BoneSpace
    alphaBoolBlend {blendInTime=0.2, blendOption=Linear, blendOutTime=0.35}
    alphaScaleBiasClamp {bInterpResult=true}
    boneToModify {boneName=cape_correct_l_01}
    rotation=<rot 3.136 -0.572 12.563>
    translation=<0.316 0.906 5.499>
   }
   n44 = "Transform (Modify) Bone - Bone: cape_correct_r_01"(ComponentPose=@n43, Alpha=1, Translation="0.353,2.429,-7.665", Rotation="0.943,3.939,15.354") {
    rotationMode=BMM_Additive
    rotationSpace=BCS_BoneSpace
    translationMode=BMM_Additive
    translationSpace=BCS_BoneSpace
    alphaBoolBlend {blendInTime=0.2, blendOption=Linear, blendOutTime=0.35}
    alphaScaleBiasClamp {bInterpResult=true}
    boneToModify {boneName=cape_correct_r_01}
    rotation=<rot 0.943 3.939 15.354>
    translation=<0.353 2.429 -7.665>
   }
   n45 = "Transform (Modify) Bone - Bone: cape_correct_r_01"!(ComponentPose=@n44, bAlphaBoolEnabled:bool=True, Translation="3.901,2.739,-4.884", Rotation="4.435,-100.697,23.871") -> (Pose) {
    alphaInputType=Bool
    rotationMode=BMM_Additive
    rotationSpace=BCS_BoneSpace
    translationMode=BMM_Additive
    translationSpace=BCS_BoneSpace
    alphaBoolBlend {blendInTime=0.2, blendOption=Linear, blendOutTime=0.35}
    alphaScaleBiasClamp {bInterpResult=true}
    boneToModify {boneName=cape_correct_r_01}
    rotation=<rot 4.435 -100.697 23.871>
    translation=<3.901 2.739 -4.884>
   }
   n46 = "Transform (Modify) Bone - Bone: cape_correct_l_01"!(ComponentPose=@n45, bAlphaBoolEnabled:bool=True, Translation="-0.512,-4.758,5.338", Rotation="-0.489,4.001,22.382") -> (Pose) {
    alphaInputType=Bool
    rotationMode=BMM_Additive
    rotationSpace=BCS_BoneSpace
    translationMode=BMM_Additive
    translationSpace=BCS_BoneSpace
    alphaBoolBlend {blendInTime=0.2, blendOption=Linear, blendOutTime=0.35}
    alphaScaleBiasClamp {bInterpResult=true}
    boneToModify {boneName=cape_correct_l_01}
    rotation=<rot -0.489 4.001 22.382>
    translation=<-0.512 -4.758 5.338>
   }
   n47 = Component To Local(ComponentPose=@n46)
   n48 = RigidBody!(
    ComponentPose=Local To Component(LocalPose=Blend Poses by bool(BlendPose_0=@n38, BlendPose_1=@n47, BlendTime_0=0.5, BlendTime_1=0.2) {blendType=Linear})
    Alpha=Get Alpha_SlidingAndKneeSliding() -> (Alpha_SlidingAndKneeSliding:double)
    OverrideWorldGravity:Vector="0,0,-1000"
    ExternalForce
    SimSpaceSettings:Sim Space Settings Structure=Get Sim Space Settings() -> (Sim Space Settings:Sim Space Settings Structure)
   ) -> (Pose) {
    bClampLinearTranslationLimitToRefPose=true
    bOverrideWorldGravity=true
    overridePhysicsAsset=&/F_MED_FreshWave_Crouching
    simulationSpace=BaseBoneSpace
    alphaBoolBlend {blendOption=Linear}
    baseBoneRef {boneName=root}
    componentAppliedLinearAccClamp=<3000 3000 1000>
    componentLinearAccScale=<0.3 0.3 0.3>
    componentLinearVelScale=<0.3 0.3 0.3>
    overrideWorldGravity {z=-1000}
    simSpaceSettings {maxAngularAcceleration=5000, maxAngularVelocity=5000, maxLinearAcceleration=5000, maxLinearVelocity=5000, worldAlpha=0.3}
   }
   n49 = Component To Local(ComponentPose=@n48)
   n50 = "Transform (Modify) Bone - Bone: cape_correct_r_01"!(
    ComponentPose=Local To Component(LocalPose=Use cached pose 'Coat_CtrlRig_Cache'())
    Alpha:float=Get Fixed_IddleFrontEnd() -> (Fixed_IddleFrontEnd:double)
    Translation="0.000,-4.317,-0.096"
    Rotation="0.093,0.017,-13.502"
    Scale:Vector="1,1,1"
   ) -> (Pose) {
    rotationMode=BMM_Additive
    rotationSpace=BCS_BoneSpace
    scaleMode=BMM_Additive
    scaleSpace=BCS_BoneSpace
    translationMode=BMM_Additive
    translationSpace=BCS_BoneSpace
    alphaBoolBlend {blendOption=Linear}
    alphaScaleBiasClamp {bInterpResult=true, interpSpeedDecreasing=5, interpSpeedIncreasing=8}
    boneToModify {boneName=cape_correct_r_01}
    rotation=<rot 0.093 0.017 -13.502>
    translation=<0 -4.3174 -0.0961>
   }
   n51 = Get Alpha_isStretchOpenParachute() -> (Alpha_isStretchOpenParachute:double)
   n52 = "Transform (Modify) Bone - Bone: dyn_cape_s_l_base"(ComponentPose=@n50, Alpha=@n51, Translation="0,-8,15", Rotation="0,0,0", Scale="1,1,1") {
    rotationMode=BMM_Additive
    rotationSpace=BCS_BoneSpace
    scaleMode=BMM_Additive
    scaleSpace=BCS_BoneSpace
    translationMode=BMM_Additive
    translationSpace=BCS_BoneSpace
    alphaBoolBlend {blendOption=Linear}
    alphaScaleBiasClamp {bInterpResult=true, interpSpeedDecreasing=5, interpSpeedIncreasing=8}
    boneToModify {boneName=dyn_cape_s_l_base}
    translation {y=-8, z=15}
   }
   n53 = "Transform (Modify) Bone - Bone: dyn_cape_s_r_base"!(ComponentPose=@n52, Alpha=@n51, Translation="0.923,-0.771,33.079", Rotation="0,0,0", Scale:Vector="1,1,1") -> (Pose) {
    rotationMode=BMM_Additive
    rotationSpace=BCS_BoneSpace
    scaleMode=BMM_Additive
    scaleSpace=BCS_BoneSpace
    translationMode=BMM_Additive
    translationSpace=BCS_BoneSpace
    alphaBoolBlend {blendOption=Linear}
    alphaScaleBiasClamp {bInterpResult=true, interpSpeedDecreasing=5, interpSpeedIncreasing=8}
    boneToModify {boneName=dyn_cape_s_r_base}
    translation=<0.923 -0.771 33.079>
   }
   n54 = "Transform (Modify) Bone - Bone: dyn_cape_b_r_base"!(ComponentPose=@n53, Alpha=@n51, Translation="12.652,-4.000,7.083", Rotation="0,0,0", Scale:Vector="1,1,1") -> (Pose) {
    rotationMode=BMM_Additive
    rotationSpace=BCS_BoneSpace
    scaleMode=BMM_Additive
    scaleSpace=BCS_BoneSpace
    translationMode=BMM_Additive
    translationSpace=BCS_BoneSpace
    alphaBoolBlend {blendOption=Linear}
    alphaScaleBiasClamp {bInterpResult=true, interpSpeedDecreasing=5, interpSpeedIncreasing=8}
    boneToModify {boneName=dyn_cape_b_r_base}
    translation=<12.652 -4 7.083>
   }
   n55 = "Transform (Modify) Bone - Bone: floor_joint"!(ComponentPose=@n54, bAlphaBoolEnabled:bool=True, Translation="0,0,-100", Rotation="0,0,0", Scale="1,1,1") -> (Pose) {
    alphaInputType=Bool
    rotationMode=BMM_Additive
    rotationSpace=BCS_BoneSpace
    scaleMode=BMM_Additive
    scaleSpace=BCS_BoneSpace
    translationMode=BMM_Additive
    translationSpace=BCS_BoneSpace
    alphaBoolBlend {blendInTime=0.1, blendOption=Linear, blendOutTime=0.1}
    alphaScaleBiasClamp {bInterpResult=true, interpSpeedDecreasing=5, interpSpeedIncreasing=8}
    boneToModify {boneName=floor_joint}
    translation {z=-100}
   }
   n56 = RigidBody(
    ComponentPose=@n55
    Alpha="Get Alphas_RB "() -> ("Alphas_RB ":double)
    OverrideWorldGravity="0,0,-1000"
    SimSpaceSettings=Get Sim Space Settings()
   ) {
    bClampLinearTranslationLimitToRefPose=true
    bOverrideWorldGravity=true
    overridePhysicsAsset=&/F_MED_FreshWave_Physics
    simulationSpace=BaseBoneSpace
    alphaBoolBlend {blendOption=Linear}
    baseBoneRef {boneName=root}
    componentAppliedLinearAccClamp=<3000 3000 1000>
    componentLinearAccScale=<0.3 0.3 0.3>
    componentLinearVelScale=<0.3 0.3 0.3>
    overrideWorldGravity {z=-1000}
    simSpaceSettings {maxAngularAcceleration=5000, maxAngularVelocity=5000, maxLinearAcceleration=5000, maxLinearVelocity=5000, worldAlpha=0.3}
   }
   n57 = Component To Local(ComponentPose=@n56)
   n58 = Get Alpha_is_Crouching Delay() -> (Alpha_is_Crouching Delay:bool)
   n59 = "Constraint - base_cape_area_l_base (P)"(ComponentPose=Local To Component(LocalPose=Use cached pose 'Input_CtrlRig_Cache'()), Alpha=1, ConstraintWeights_0=1) {
    alphaBoolBlend {blendOption=Linear}
    boneToModify {boneName=base_cape_area_l_base}
    constraintSetup [{targetBone {boneName=thigh_l}}]
    constraintWeights [1]
   }
   n60 = "Constraint - base_cape_area_r_base (P)"(ComponentPose=@n59, Alpha=1, ConstraintWeights_0=1) {
    alphaBoolBlend {blendOption=Linear}
    boneToModify {boneName=base_cape_area_r_base}
    constraintSetup [{targetBone {boneName=thigh_r}}]
    constraintWeights [1]
   }
   n61 = RigidBody(ComponentPose=@n60, OverrideWorldGravity="0,0,-1000", SimSpaceSettings=Get Sim Space Settings()) {
    bClampLinearTranslationLimitToRefPose=true
    bOverrideWorldGravity=true
    overridePhysicsAsset=&/F_MED_FreshWave_Physics_SkydivingFrontEnd
    simulationSpace=BaseBoneSpace
    alphaBoolBlend {blendOption=Linear}
    baseBoneRef {boneName=root}
    componentAppliedLinearAccClamp=<3000 3000 1000>
    componentLinearAccScale=<0.3 0.3 0.3>
    componentLinearVelScale=<0.3 0.3 0.3>
    overrideWorldGravity {z=-1000}
    simSpaceSettings {maxAngularAcceleration=5000, maxAngularVelocity=5000, maxLinearAcceleration=5000, maxLinearVelocity=5000, worldAlpha=0.3}
   }
   Save cached pose 'RigidBody Cache'(Pose=Blend Poses by bool(
     BlendPose_0=Component To Local(ComponentPose=@n61)
     BlendPose_1=Blend Poses by bool(BlendPose_0=@n49, BlendPose_1=@n57, BlendTime_0=0.2, BlendTime_1=0.2, bActiveValue=@n58) {blendType=Linear}
     bActiveValue=Get Bool_SkydivingOnFrontEnd() -> (Bool_SkydivingOnFrontEnd:bool)
    ) {blendType=Linear}) {cacheName=RigidBody Cache}
   n62 = Control Rig!(Source=Copy Pose From Mesh() {bUseAttachedParent=true}, Alpha, Selected_Pistol:double=Get Selected_Pistol() -> (Selected_Pistol:double)) -> (Pose) {
    alphaBoolBlend {blendOption=Linear}
    controlRigAssetReference {blueprintRigClass=&/F_MED_FreshWave_CtrlRig.F_MED_FreshWave_CtrlRig_C}
   }
   Save cached pose 'Input_CtrlRig_Cache'(Pose=@n62) {cacheName=Input_CtrlRig_Cache}
  ]}
 EventGraph {nodes [
   n1 = Event BlueprintUpdateAnimation()
   > Set Alpha_DeltaTimeX(Alpha_DeltaTimeX:double=@n1.DeltaTimeX) -> (Output_Get:double)
   > n2 = Sequence!(execute) -> (
    then_0
    then_1
    then_2:Exec
    then_3:Exec
    then_4:Exec
    then_5:Exec
    then_6:Exec
    then_7:Exec
    then_8:Exec
    then_9:Exec
    then_10:Exec
    then_11:Exec
   )
   n3 = BreakTransform(InTransform=GetDeltaTransformFromRefPose(self=GetOwnerSkeletalMeshComponent(), BoneName=spine_04, BaseName=spine_03))
   > n4 = Branch(Condition="float > float"(A=BreakVector(InVec=@n3), B=8))
   > Set Alpha_isStretchOpenParachute(Alpha_isStretchOpenParachute:double=1) -> (Output_Get:double)
   Set Alpha_isStretchOpenParachute(execute=@n4.else, Alpha_isStretchOpenParachute=0)
   n5 = Branch(execute=@n2.then_4, Condition=Get bIsInFrontEnd())
   > n6 = Branch(Condition=Get bIsPlayingEmote())
   > Set Sim Space Settings(Sim Space Settings:Sim Space Settings Structure=Get Sim Space Settings_isFrontEndEmote() -> (Sim Space Settings_isFrontEndEmote:Sim Space Settings Structure)) -> (Output_Get:Sim Space Settings Structure)
   Set Sim Space Settings(execute=@n6.else, Sim Space Settings=Get Sim Space Settings_isFrontEnd() -> (Sim Space Settings_isFrontEnd:Sim Space Settings Structure))
   n7 = Branch(execute=@n5.else, Condition=Get bIsSkyDiving())
   > Set Sim Space Settings(Sim Space Settings=Get Sim Space Settings_isSkydivingTrue() -> (Sim Space Settings_isSkydivingTrue:Sim Space Settings Structure))
   n8 = Branch(execute=@n7.else, Condition=Get bIsDBNOCarried())
   > Set Sim Space Settings(Sim Space Settings=Get Sim Space Settings_isDBNOCarried() -> (Sim Space Settings_isDBNOCarried:Sim Space Settings Structure))
   n9 = Branch(execute=@n8.else, Condition=Get bIsDBNO())
   > Set Sim Space Settings(Sim Space Settings=Get Sim Space Settings_isDBNO() -> (Sim Space Settings_isDBNO:Sim Space Settings Structure))
   Set Sim Space Settings(execute=@n9.else, Sim Space Settings=Get Sim Space Settings_isStanding() -> (Sim Space Settings_isStanding:Sim Space Settings Structure))
   n10 = Branch(execute=@n2.then_8, Condition=AND Boolean(A=Get bIsSkyDiving(), B=Get bIsInFrontEnd()))
   > Set Bool_SkydivingOnFrontEnd(Bool_SkydivingOnFrontEnd:bool=true) -> (Output_Get:bool)
   Set Bool_SkydivingOnFrontEnd(execute=@n10.else, Bool_SkydivingOnFrontEnd=false)
   n11 = GetAnimInstance(self=GetOwnerSkeletalMeshComponent())
   Set PlayingBS(
    execute=@n2.then_11
    PlayingBS:double="To Float (Boolean)"(InBool=Montage_IsPlaying(self=@n11, Montage=/BRCosmetics/Animation/Game/MainPlayer/Emotes/BalletSpin/Emote_BalletSpin_CMF_M))
   ) -> (Output_Get:double)
   n12 = PlayerMovement(execute=@n2.then_1, AnimInstance=Self-Reference(), XDotProductCoefficient=100, YDotProductCoefficient=100, ZDotProductCoefficient=100)
   n13 = BreakVector(InVec=MakeVector(X=@n12, Y=@n12.YMovement, Z=@n12.ZMovement))
   > Set Left(Left:bool="float < float"(A=@n13, B=-50)) -> (Output_Get:bool)
   > Set Right(Right:bool="float > float"(A=@n13, B=50)) -> (Output_Get:bool)
   > Set Forward(Forward:bool="float > float"(A=@n13.Y, B=50)) -> (Output_Get:bool)
   > Set Backward(Backward:bool="float < float"(A=@n13.Y, B=-50)) -> (Output_Get:bool)
   n14 = Branch(execute=@n2.then_2, Condition=Get bIsInFrontEnd())
   > Set Alpha_isFrontEnd / BR Necklace(Alpha_isFrontEnd / BR Necklace:double=0.5) -> (Output_Get:double)
   Set Alpha_isFrontEnd / BR Necklace(execute=@n14.else, Alpha_isFrontEnd / BR Necklace=1)
   n15 = Branch(execute=@n2.then_3, Condition=OR Boolean(A=Get bIsSliding(), B=Get bIsKneeSliding()))
   > Set Alpha_SlidingAndKneeSliding(Alpha_SlidingAndKneeSliding:double=0.1) -> (Output_Get:double)
   Set Alpha_SlidingAndKneeSliding(execute=@n15.else, Alpha_SlidingAndKneeSliding=0.8)
   n16 = Get bIsMoving2D()
   Set Alpha_Moving2d(
    execute=@n2.then_5
    Alpha_Moving2d:double=FInterpTo(
     Current=Get Alpha_Moving2d()
     Target="To Float (Boolean)"(InBool=@n16)
     DeltaTime=Get Alpha_DeltaTimeX() -> (Alpha_DeltaTimeX:double)
     InterpSpeed=2
    )
   ) -> (Output_Get:double)
   n17 = Get bIsMoving2D()
   n18 = Get bIsSkyDiving()
   n19 = Get bIsPlayingEmote()
   n20 = Branch(
    execute=@n2.then_9
    Condition=AND Boolean!(A=Get bIsInFrontEnd(), B=NOT Boolean(A=@n17), C:bool=NOT Boolean(A=@n18), D:bool=NOT Boolean(A=@n19)) -> (ReturnValue)
   )
   > Set Fixed_IddleFrontEnd(Fixed_IddleFrontEnd:double=1) -> (Output_Get:double)
   Set Fixed_IddleFrontEnd(execute=@n20.else, Fixed_IddleFrontEnd=0)
   n21 = AND Boolean!(
    A=Get Forward() -> (Forward:bool)
    B=NOT Boolean(A=Get bIsSkyDiving())
    C=NOT Boolean(A=Get bIsDBNO())
    D=NOT Boolean(A=Get bIsSurfaceSwimming())
    E:bool=NOT Boolean(A=Get bRigidBodyEnabled())
   ) -> (ReturnValue)
   Set Alpha_is_Forward(
    execute=@n2.then_7
    Alpha_is_Forward:double=FInterpTo(Current=Get Alpha_is_Forward(), Target="To Float (Boolean)"(InBool=@n21), DeltaTime=Get Alpha_DeltaTimeX(), InterpSpeed=10)
   ) -> (Output_Get:double)
   n22 = Branch(execute=@n2.then_6, Condition=Get OnZipline() -> (OnZipline:bool))
   > "Set Alphas_RB "("Alphas_RB ":double=0.3) -> (Output_Get:double)
   "Set Alphas_RB "(execute=@n22.else, "Alphas_RB "=0.7)
   n23 = Get bIsOnGround() -> (bIsOnGround:bool)
   n24 = GetAnimInstance(self=GetOwnerSkeletalMeshComponent())
   n25 = Branch(
    execute=@n2.then_10
    Condition=AND Boolean!(
     A=NOT Boolean(A=@n23)
     B=Montage_IsPlaying(self=@n24, Montage=/BRCosmetics/Animation/Game/MainPlayer/Emotes/LogarithmKick/Emote_LogarithmKick_CMF_M)
    ) -> (ReturnValue)
   )
   > Set Trinity Kick FE(Trinity Kick FE:bool=true) -> (Output_Get:bool)
   Set Trinity Kick FE(execute=@n25.else, Trinity Kick FE=false)
   n26 = Event BlueprintProcessAnimInputEvent()
   > n27 = Branch(Condition=IsNamed(self=@n26.AnimInputEvent, InEventName=aie.Player.Zipline.Entered))
   > Set OnZipline(OnZipline:bool=true) -> (Output_Get:bool)
   > Set OnZipline_Float(OnZipline_Float:double=1) -> (Output_Get:double)
   Branch(execute=@n27.else, Condition=IsNamed(self=@n26.AnimInputEvent, InEventName=aie.Player.Zipline.Exited))
   > Set OnZipline(OnZipline=false)
   > Set OnZipline_Float(OnZipline_Float=0)
   Event OnWeaponChanged()
   > n28 = Cast To FortPlayerAnimInstance(Object=GetAnimInstance(self=GetOwnerSkeletalMeshComponent()))
   > Set Selected_Pistol(Selected_Pistol:double=0) -> (Output_Get:double)
   > n29 = Switch on EFortWeaponCoreAnimation!(execute, Selection:EFortWeaponCoreAnimation Enum=Get WeaponCoreAnim(self=@n28)) -> (
    Melee:Exec
    Pistol:Exec
    Shotgun:Exec
    PaperBlueprint:Exec
    Rifle:Exec
    MeleeOneHand:Exec
    MachinePistol:Exec
    RocketLauncher:Exec
    GrenadeLauncher:Exec
    GoingCommando:Exec
    AssaultRifle:Exec
    TacticalShotgun:Exec
    SniperRifle:Exec
    TrapPlacement:Exec
    ShoulderLauncher:Exec
    AbilityDecoTool:Exec
    Crossbow:Exec
    C4:Exec
    RemoteControl:Exec
    DualWield:Exec
    AR_BullPup:Exec
    AR_ForwardGrip:Exec
    MedPackPaddles:Exec
    SMG_P90:Exec
    AR_DrumGun:Exec
    Consumable_Small:Exec
    Consumable_Large:Exec
    Balloon:Exec
    MountedTurret:Exec
    CreativeTool:Exec
    ExplosiveBow:Exec
    AshtonIndigo:Exec
    AshtonChicago:Exec
    MeleeDualWield:Exec
    Unarmed:Exec
   )
   Set Selected_Pistol(execute=@n29.Pistol, Selected_Pistol=1)
   Event OnBeginCrouching()
   > Set Alpha_is_Crouching Delay(Alpha_is_Crouching Delay:bool=1) -> (Output_Get:bool)
   Event OnEndCrouching()
   > Delay(Duration=0.2)
   > Set Alpha_is_Crouching Delay(Alpha_is_Crouching Delay=false)
  ]}
]
variables [
 Alpha_DeltaTimeX {type=number}
 Alpha_is_BR / Emote {type=number}
 "Alpha_is_BR/Emote - Fur Constrain " {type=number}
 Alpha_is_Crouching Delay {type=boolean}
 Alpha_is_DBNO_Carried {type=number}
 Alpha_is_Detected_Rot_UP_Hand_L {type=number}
 Alpha_is_Detected_Rot_UP_Hand_R {type=number}
 Alpha_is_DetectedRot_ClavicleL_Emote {type=number}
 Alpha_is_DetectedRot_ClavicleR_Emote {type=number}
 Alpha_is_DetectedRot_Hand_L {type=number}
 Alpha_is_DetectedRot_Hand_R {type=number}
 Alpha_is_Forward {type=number}
 Alpha_is_GrindRail {type=boolean}
 Alpha_is_Ground/SkyDiving {type=number}
 Alpha_is_Ground_DBNO {type=number}
 Alpha_is_Jump - Moving2d {type=number}
 Alpha_is_Jump -Moving2d 02 {type=number}
 Alpha_is_Jump-Idle {type=number}
 Alpha_is_Moving2d On Ground {type=number}
 Alpha_is_Open Parachute Umbrella {type=number}
 Alpha_is_Stretch Open Parachute Umbrella {type=number}
 Alpha_isCarryingStart {type=number}
 Alpha_isDifferentsGliders / BR game {type=number}
 Alpha_isFloating / SkydiveMode {type=number}
 Alpha_isFrontEnd / BR Necklace {type=number}
 Alpha_isHeadRotateL Melee {type=number}
 Alpha_isHeadRotateR Melee {type=number}
 Alpha_isReviveDBNO {type=boolean}
 Alpha_isStretchOpenParachute {type=number}
 Alpha_MeleeJumping {type=number}
 Alpha_Moving2d {type=number}
 Alpha_SlidingAndKneeSliding {type=number}
 "Alphas_RB " {type=number}
 Backward {type=boolean}
 Bool_SkydivingOnFrontEnd {type=boolean}
 Detected Leg Rotation Corrective R {type=number}
 Emote_BalletSpin {type=number}
 Fixed_IddleFrontEnd {type=number}
 Forward {type=boolean}
 isStrafe_L {type=number}
 Left {type=boolean}
 Left_leg_rot {type=boolean}
 Node {type=LinkedAnimGraphReference, default {}}
 OnBackwardOnZipline {type=number}
 OnZipline {type=boolean}
 OnZipline_Float {type=number}
 PlayingBS {type=number}
 Right {type=boolean}
 Selected_Pistol {type=number}
 Sim Space Settings {
  type=SimSpaceSettings
  default {
   dampingAlpha=1
   maxAngularAcceleration=10000
   maxAngularVelocity=3
   maxLinearAcceleration=10000
   maxLinearVelocity=10000
   velocityScaleZ=1
   worldAlpha=0
   externalLinearDragV=<0 0 0>
  }
 }
 Sim Space Settings_isDBNO {
  type=SimSpaceSettings
  default {
   dampingAlpha=1
   maxAngularAcceleration=6
   maxAngularVelocity=1000
   maxLinearAcceleration=500
   maxLinearVelocity=1000
   velocityScaleZ=1
   worldAlpha=0.25
   externalLinearDragV=<0 0 0>
  }
 }
 Sim Space Settings_isDBNOCarried {
  type=SimSpaceSettings
  default {
   dampingAlpha=1
   maxAngularAcceleration=7
   maxAngularVelocity=10
   maxLinearAcceleration=7
   maxLinearVelocity=10
   velocityScaleZ=1
   worldAlpha=0.2
   externalLinearDragV=<0 0 0>
  }
 }
 Sim Space Settings_isFrontEnd {
  type=SimSpaceSettings
  default {
   dampingAlpha=1
   maxAngularAcceleration=6
   maxAngularVelocity=1000
   maxLinearAcceleration=500
   maxLinearVelocity=1000
   velocityScaleZ=1
   worldAlpha=0.25
   externalLinearDragV=<0 0 0>
  }
 }
 Sim Space Settings_isFrontEndEmote {
  type=SimSpaceSettings
  default {
   dampingAlpha=1
   maxAngularAcceleration=6
   maxAngularVelocity=1000
   maxLinearAcceleration=500
   maxLinearVelocity=1000
   velocityScaleZ=1
   worldAlpha=0.25
   externalLinearDragV=<0 0 0>
  }
 }
 Sim Space Settings_isSkydivingFalse {
  type=SimSpaceSettings
  default {
   dampingAlpha=1
   maxAngularAcceleration=600
   maxAngularVelocity=1000
   maxLinearAcceleration=500
   maxLinearVelocity=1000
   velocityScaleZ=1
   worldAlpha=0.1
   externalLinearDragV=<0.2 0.2 0>
  }
 }
 Sim Space Settings_isSkydivingTrue {
  type=SimSpaceSettings
  default {
   dampingAlpha=1
   maxAngularAcceleration=600
   maxAngularVelocity=3
   maxLinearAcceleration=500
   maxLinearVelocity=1000
   velocityScaleZ=1
   worldAlpha=0.2
   externalLinearDragV=<0.2 0.2 0>
  }
 }
 Sim Space Settings_isStanding {
  type=SimSpaceSettings
  default {
   dampingAlpha=1
   maxAngularAcceleration=6
   maxAngularVelocity=500
   maxLinearAcceleration=500
   maxLinearVelocity=500
   velocityScaleZ=1
   worldAlpha=0.25
   externalLinearDragV=<0 0 0>
  }
 }
 Standing Fixes Coat {type=number}
 Trinity Kick FE {type=boolean}
]
