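# Physics asset of collision capsules for the F_MED_FreshWave female medium player body cosmetic.
# Single capsules cover the pelvis, lower spine bones, upper arms, thighs and calves, plus selected front and back dyn_cape bones on both sides.
# Only three constraints exist, linking right upper arm to spine_02, right thigh to pelvis and right calf to thigh.
# All three leave swing and twist free, so their angle limits are inactive.
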PhysicsAsset /BRCosmetics/Characters/Player/Female/Medium/Bodies/F_MED_FreshWave/Meshes/F_MED_FreshWave_Colliders
bodies {
 calf_l [{shapeType=Capsule, length=26, radius=15, shapeName=calf_l_capsule, center=<-31 0 0>, rotation=<rot 0 90 90>}]
 calf_r [{shapeType=Capsule, length=26, radius=15, shapeName=calf_l_capsule, center=<31 0 0>, rotation=<rot 0 90 -90>}]
 dyn_cape_b_l_02 [{shapeType=Capsule, length=15, radius=17, shapeName=dyn_cape_b_l_02_capsule, center=<9.1 4 -6.3>, rotation=<rot 0 90 90>}]
 dyn_cape_b_l_04 [{
   shapeType=Capsule
   length=22
   radius=21
   shapeName=dyn_cape_b_l_04_capsule
   center=<2.4173 9.3666 -8.8943>
   rotation=<rot -0.0001 0 105.5321>
  }]
 dyn_cape_b_r_02 [{shapeType=Capsule, length=15, radius=17, shapeName=dyn_cape_b_r_02_capsule, center=<9.1 -4 -6.3>, rotation=<rot 0 90 90>}]
 dyn_cape_b_r_04 [{shapeType=Capsule, length=22, radius=21, shapeName=dyn_cape_b_r_04_capsule, center=<2.4173 -6.7206 -8.6504>, rotation=<rot 0 0 -105.5321>}]
 dyn_cape_f_l_01 [{shapeType=Capsule, length=36, radius=11, shapeName=dyn_cape_f_l_01_capsule, center=<34.9949 0 0>, rotation=<rot 0 90 90>}]
 dyn_cape_f_r_01 [{shapeType=Capsule, length=36, radius=11, shapeName=dyn_cape_f_r_01_capsule, center=<34.9949 0 0>, rotation=<rot 0 90 90>}]
 pelvis [{shapeType=Capsule, length=11, radius=17, shapeName=pelvis_capsule, center=<0 -0.6439 0>}]
 spine_01 [{shapeType=Capsule, length=11, radius=15, shapeName=spine_01_capsule, center=<4.8717 0 0>}]
 spine_02 [{shapeType=Capsule, length=15, radius=17, shapeName=spine_02_capsule, center=<23.61 0 0>, rotation=<rot 0 90 90>}]
 thigh_l [{shapeType=Capsule, length=26.5, radius=17, shapeName=thigh_l_capsule, center=<-26.493 -0.642 0.4474>, rotation=<rot 0 90 90>}]
 thigh_r [{shapeType=Capsule, length=26.5, radius=17, shapeName=thigh_l_capsule, center=<26.493 0.642 -0.4474>, rotation=<rot 0 90 -90>}]
 upperarm_l [{shapeType=Capsule, length=15, radius=10.5, shapeName=upperarm_l_capsule, center=<9.7775 0 -1.0344>, rotation=<rot 0 90 90>}]
 upperarm_r [{shapeType=Capsule, length=15, radius=10.5, shapeName=upperarm_l_capsule, center=<-9.7775 0 1.0344>, rotation=<rot 0 90 -90>}]
}
constraints [
 {
  bone1Name=upperarm_r
  bone2Name=spine_02
  swing1LimitDegrees=45
  swing1Motion=Free
  swing2LimitDegrees=45
  swing2Motion=Free
  twistLimitDegrees=45
  twistMotion=Free
 }
 {
  bone1Name=thigh_r
  bone2Name=pelvis
  swing1LimitDegrees=45
  swing1Motion=Free
  swing2LimitDegrees=45
  swing2Motion=Free
  twistLimitDegrees=45
  twistMotion=Free
 }
 {
  bone1Name=calf_r
  bone2Name=thigh_r
  swing1LimitDegrees=45
  swing1Motion=Free
  swing2LimitDegrees=45
  swing2Motion=Free
  twistLimitDegrees=45
  twistMotion=Free
 }
]
