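# Front-end physics asset for the FreshWave medium female character's face accessory mesh, covering two hair strands, the head and fur.
# Left and right four-bone hair chains hang from the head, with tight, slightly asymmetric swing limits and locked twist to restrict bending.
# Capsules on the head, upper spine and both fur bones supply body volumes around the hair.
# The head and fur bodies link to spine_05 through fully free angular constraints.
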PhysicsAsset /BRCosmetics/Characters/Player/Female/Medium/Bodies/F_MED_FreshWave/Meshes/Parts/F_MED_FreshWave_FaceAcc_Physics_FrontEnd
bodies {
 base_fur_02_l [{
   shapeType=Capsule
   length=11.2203
   radius=4.8378
   shapeName=base_fur_02_l_capsule
   center=<0 0.6416 -0.7887>
   rotation=<rot 0.0002 0 -17.8409>
  }]
 base_fur_02_r [{shapeType=Capsule, length=9, radius=5.5, shapeName=base_fur_02_l_capsule, center=<0 -0.6 1.1>, rotation=<rot 0.0002 0 171.9272>}]
 hair_A_01_l [{shapeType=Capsule, length=2, radius=2, shapeName=hair_A_01_l_capsule, rotation=<rot 90 0 0>}]
 hair_A_01_r [{shapeType=Capsule, length=2, radius=2, shapeName=hair_A_01_r_capsule, rotation=<rot 90 0 0>}]
 hair_A_02_l [{shapeType=Capsule, length=2, radius=2, shapeName=hair_A_02_l_capsule, rotation=<rot 90 0 0>}]
 hair_A_02_r [{shapeType=Capsule, length=2, radius=2, shapeName=hair_A_02_r_capsule, rotation=<rot 90 0 0>}]
 hair_A_03_l [{shapeType=Capsule, length=2, radius=1.5, shapeName=hair_A_03_l_capsule, rotation=<rot 90 0 0>}]
 hair_A_03_r [{shapeType=Capsule, length=2, radius=1.8, shapeName=hair_A_03_r_capsule, rotation=<rot 90 0 0>}]
 hair_A_04_l [{
   shapeType=Capsule
   length=3.6014
   radius=1.1013
   shapeName=hair_A_04_l_capsule
   center=<0.9021 0.2266 0.0691>
   rotation=<rot 72.6479 -171.6192 -173.6293>
  }]
 hair_A_04_r [{
   shapeType=Capsule
   length=4.5
   radius=1.6
   shapeName=hair_A_04_l_capsule
   center=<-1.0356 -0.5 -1.2>
   rotation=<rot 77.6387 -171.6085 -0.0002>
  }]
 head [
  {shapeType=Capsule, length=10, radius=4.5, shapeName=head_capsule, center=<-2.664 0.758 -0.8928>, rotation=<rot 90 0 0>}
  {shapeType=Capsule, length=15, radius=8, shapeName=head_capsule, center=<-6.9835 -7.3326 -9.0319>}
 ]
 spine_05 [{shapeType=Capsule, length=5, radius=5, shapeName=spine_05_capsule, rotation=<rot 90 0 0>}]
}
constraints [
 {bone1Name=hair_A_01_l, bone2Name=head, swing1LimitDegrees=7, swing2LimitDegrees=7, twistLimitDegrees=45, twistMotion=Locked}
 {bone1Name=hair_A_02_l, bone2Name=hair_A_01_l, swing1LimitDegrees=7, swing2LimitDegrees=15, twistLimitDegrees=45, twistMotion=Locked}
 {bone1Name=hair_A_03_l, bone2Name=hair_A_02_l, swing1LimitDegrees=7, swing2LimitDegrees=15, twistLimitDegrees=45, twistMotion=Locked}
 {bone1Name=hair_A_04_l, bone2Name=hair_A_03_l, swing1LimitDegrees=10, swing2LimitDegrees=12, twistLimitDegrees=45, twistMotion=Locked}
 {bone1Name=hair_A_01_r, bone2Name=head, swing1LimitDegrees=3.5, swing2LimitDegrees=7, twistLimitDegrees=45, twistMotion=Locked}
 {bone1Name=hair_A_03_r, bone2Name=hair_A_02_r, swing1LimitDegrees=7, swing2LimitDegrees=15, twistLimitDegrees=45, twistMotion=Locked}
 {bone1Name=hair_A_02_r, bone2Name=hair_A_01_r, swing1LimitDegrees=5, swing2LimitDegrees=10, twistLimitDegrees=45, twistMotion=Locked}
 {bone1Name=hair_A_04_r, bone2Name=hair_A_03_r, swing1LimitDegrees=7, swing2LimitDegrees=15, twistLimitDegrees=45, twistMotion=Locked}
 {
  bone1Name=head
  bone2Name=spine_05
  swing1LimitDegrees=45
  swing1Motion=Free
  swing2LimitDegrees=45
  swing2Motion=Free
  twistLimitDegrees=45
  twistMotion=Free
 }
 {
  bone1Name=base_fur_02_l
  bone2Name=spine_05
  swing1LimitDegrees=45
  swing1Motion=Free
  swing2LimitDegrees=45
  swing2Motion=Free
  twistLimitDegrees=45
  twistMotion=Free
 }
 {
  bone1Name=base_fur_02_r
  bone2Name=spine_05
  swing1LimitDegrees=45
  swing1Motion=Free
  swing2LimitDegrees=45
  swing2Motion=Free
  twistLimitDegrees=45
  twistMotion=Free
 }
]
